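# Translucent deferred decal material for the HighTower Date runes effect, with green emissive glyphs over a ground char mark.
# Radially mapped, time-scrolled fire noise modulates the rune texture, while tiling noise and a LifeCurve reroute drive threshold erosion of glow and opacity.
# LifeCurve comes from decal lifetime opacity, with a static switch to a debug sine for previewing fades.
# UV screen derivatives (DDX/DDY) build an opacity mask against UV discontinuities or stretching.
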Material /TidalCrane_Swarm_Boss/Assets/FX/EventVariant/Projectile/MD_HighTower_Date_Runes
graph {
 properties {
  bAutomaticallySetUsageInEditor=false
  blendMode=BLEND_Translucent
  materialDomain=MD_DeferredDecal
  pixelDepthOffsetMode=PDOM_Legacy
  usedShadingModels=Default Lit
 }
 nodes [
  Constant()
  n1 = VectorParameter() {parameterName=EmissiveColor, defaultValue=<rgba 0.4424 1 0.0908 1>}
  n2 = OneMinus(None=NamedRerouteUsage())
  n3 = MaterialFunctionCall(Float2=MaterialFunctionCall() {materialFunction=&VectorToRadialValue}.Radial Coordinates) {materialFunction=&BreakOutFloat2Components}
  n4 = Divide(A=@n3.G) {constB=4}
  n5 = Divide(A=Time()) {constB=-8}
  n6 = TextureSample(UVs=AppendVector(A=@n3.R, B=Add(A=@n4, B=@n5))) {texture=/Game/Global/Textures/Default/Noise/T_FireNoise_02}
  n7 = Multiply(
   A=Multiply(A=TextureSample() {samplerType=SAMPLERTYPE_LinearGrayscale, texture=/Game/Effects/Fort_Effects/Textures/Decal/T_HightTower_Date_Runes}.R) {constB=1.7}
   B=Multiply(A=@n6.R, B=@n6.R)
  )
  n8 = Multiply(
   A=LinearInterpolate(A=LinearInterpolate(Alpha=@n2), Alpha=OneMinus(None=@n7)) {constA=0.1}
   B=Multiply(A=@n2, B=TextureSample() {texture=/Game/Global/Textures/Default/Noise/T_Tiling_Noise_01}.R)
  )
  n9 = LinearInterpolate(A=LinearInterpolate(Alpha=NamedRerouteUsage()) {constA=1, constB=0.1}, Alpha=@n8) {constB=80}
  LinearInterpolate(
   B=LinearInterpolate(A=Divide(A=@n1.RGB) {constB=50}, B=Multiply(A=@n1.RGB) {constB=5}, Alpha=NamedRerouteUsage())
   Alpha=Saturate(None=Subtract(A=@n7, B=@n9))
  )
  n10 = TextureCoordinate()
  n11 = DDX(Value=@n10)
  n12 = Constant() {r=1000000}
  n13 = LinearInterpolate(B=@n12, Alpha=ScalarParameter() {defaultValue=1})
  n14 = Multiply(A=Subtract(A=Abs(None=@n11), B=Divide(B=@n13) {constB=10000}), B=@n13) {constB=5000}
  n15 = Abs(None=DDY(Value=@n10))
  n16 = Divide(B=@n12) {constB=10000}
  Min(A=@n14, B=Multiply(A=Constant() {r=1000000}, B=Subtract(A=@n15, B=@n16)) {constB=5000})
  n17 = MaterialFunctionCall(Float2=@n14) {materialFunction=&BreakOutFloat2Components}
  n18 = LinearInterpolate(Alpha=NamedRerouteUsage()) {constA=1, constB=0.1}
  n19 = Multiply(A=Subtract(A=@n7, B=LinearInterpolate(A=@n18, Alpha=@n8) {constA=0.2, constB=10}), B=LinearInterpolate(Alpha=@n2) {constA=1, constB=3})
  n20 = TextureSample(UVs=MaterialFunctionCall(Texture Scale=Constant() {r=1.5}) {materialFunction=&ScaleUVsByCenter}.UVs) {samplerType=SAMPLERTYPE_LinearColor, texture=/Game/Effects/Fort_Effects/Gameplay/SmokeDustExplosions/Textures/T_GroundChar_01}
  n21 = NamedRerouteUsage()
  Multiply(A=Saturate(None=Min(A=@n17.R, B=@n17.R)), B=Add(A=Saturate(None=@n19), B=Multiply(A=@n20.R, B=@n21)))
  NamedRerouteDeclaration(None=StaticSwitch(
    True=MaterialFunctionCall() {materialFunction=&DebugTimeSine}.Result
    False=DecalLifetimeOpacity().Opacity
    Value=StaticBool()
   ) {desc=PreviewDecalFading}) {name=LifeCurve, nodeColor=<rgba 1 0.0938 0 1>}
  ParticleColor()
 ]
}
