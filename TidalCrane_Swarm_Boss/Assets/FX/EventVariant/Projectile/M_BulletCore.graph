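# Translucent unlit Niagara mesh-particle material for the core of an event-variant Tidal Crane swarm boss projectile.
# Color layers a fake GGX highlight from a rock-edge normal over gradient-tinted, texture-masked colors and noise-driven soft glow.
# Opacity combines particle alpha with a rock-texture dissolve weighted along UV V.
# World position offset pushes vertices along normals by that texture and bends them around particle position via dynamic parameters and local-space distance.
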Material /TidalCrane_Swarm_Boss/Assets/FX/EventVariant/Projectile/M_BulletCore
graph {
 properties {
  bAutomaticallySetUsageInEditor=false
  bUsedWithNiagaraMeshParticles=true
  blendMode=BLEND_Translucent
  shadingModel=MSM_Unlit
  usedShadingModels=Unlit
 }
 nodes [
  ActorPositionWS()
  n1 = ScalarParameter() {defaultValue=0.3025, parameterName=R}
  n2 = VectorParameter() {group=MainTexture, parameterName=MainTexture_UVControl, channelNames {a=VPan, b=UPan, g=VScale, r=UScale}, defaultValue=<rgba 1 1 0 0.4>}
  n3 = MaterialFunctionCall(
   Texture=TextureObject() {texture=&/T_RockEdgeSoft}
   UVControl_RGB=@n2
   UVControl_A=@n2
  ) {materialFunction=&FXCommon_TextureSample}
  n4 = Transform(None=TextureSample(UVs=@n3.UV) {samplerType=SAMPLERTYPE_Normal, texture=&/T_RockEdgeNormal}.RGB)
  n5 = Add(
   A=Saturate(None=MaterialFunctionCall!(LightDirctionIn, CameraVectorIn, Roughness=@n1, NormalIn=@n4) {materialFunction=/Game/Effects/Niagara/MasterMaterials/Functions/Common/FXCommon_GGX}.Result)
   B=ScalarParameter() {group=HighLight, parameterName=HighLight_GrayAdjust_Add}
  )
  n6 = ScalarParameter() {defaultValue=4.3967, group=HighLight, parameterName=HighLight_GrayAdjust_Mul}
  n7 = Power(
   Base=Saturate(None=Multiply(A=@n5, B=@n6))
   Exp=ScalarParameter() {defaultValue=0.8477, group=HighLight, parameterName=HighLight_GrayAdjust_Power}
  )
  n8 = MaterialFunctionCall(Float2=TextureCoordinate()) {materialFunction=&BreakOutFloat2Components}
  n9 = VectorParameter() {parameterName=TipBright_Refine, defaultValue=<rgba 12.7868 41.4893 0 0.5>}
  n10 = Multiply(
   A=Add(A=MaterialFunctionCall(Texture=@n8.G, RGB=@n9.RGB, A=@n9.RGB) {materialFunction=&MF_FX_Refine}.Result)
   B=VectorParameter() {parameterName=DC, defaultValue=<rgba 0.4817 0.0607 0.6348 1>}.RGB
  )
  n11 = VectorParameter() {parameterName=BC, defaultValue=<rgba 8 0 0.9333 1>}
  n12 = TextureObjectParameter() {
   group=MainTexture
   parameterName=MainTexture_Texture
   sortPriority=0
   texture=&/T_RockEdge
  }
  n13 = VectorParameter() {parameterName=MainTexture_Refine, defaultValue=<rgba 1 1 1 0.5>}
  n14 = MaterialFunctionCall(
   Texture=OneMinus(None=MaterialFunctionCall(Texture=@n12, UVControl_RGB=@n2, UVControl_A=@n2) {materialFunction=&FXCommon_TextureSample}.R)
   RGB=@n13.RGB
   A=@n13.RGB
  ) {materialFunction=&MF_FX_Refine}
  n15 = TextureObjectParameter() {
   group=NoiseForEmissive
   parameterName=NoiseForEmissive_Texture
   samplerType=SAMPLERTYPE_LinearColor
   sortPriority=0
   texture=/Game/Global/Textures/Default/Noise/T_Noise01_Inverse
  }
  n16 = VectorParameter() {
   group=NoiseForEmissive
   parameterName=NoiseForEmissive_UVControl
   channelNames {a=VPan, b=UPan, g=VScale, r=UScale}
   defaultValue=<rgba 1 0.5 0 0.8>
  }
  n17 = VectorParameter() {parameterName=SoftC_Refine, defaultValue=<rgba 1 5.4951 1 0.5>}
  n18 = MaterialFunctionCall(
   Texture=Multiply(
    A=MaterialFunctionCall(Texture=@n15, UVControl_RGB=@n16, UVControl_A=@n16) {materialFunction=&FXCommon_TextureSample}.R
    B=OneMinus(None=@n3.R)
   )
   RGB=@n17.RGB
   A=@n17.RGB
  ) {materialFunction=&MF_FX_Refine}
  n19 = VectorParameter() {parameterName=SoftC, defaultValue=<rgba 2.2434 0 3 1>}
  Add(
   A=Multiply(A=Saturate(None=@n7), B=VectorParameter() {parameterName=RefC, defaultValue=<rgba 0.7737 0.3652 1 1>}.RGB)
   B=Add(A=LinearInterpolate(A=@n10, B=@n11.RGB, Alpha=@n14.Result), B=Multiply(A=@n18.Result, B=@n19.RGB))
  )
  Constant3Vector()
  DotProduct(B=@n4)
  n20 = TextureCoordinate()
  n21 = LinearInterpolate(
   A=ScalarParameter() {defaultValue=-1.0017, group=DissolveAtEnd, parameterName=DissolveAtEnd_GrayAdjust_Add2}
   B=ScalarParameter() {defaultValue=1, group=DissolveAtEnd, parameterName=DissolveAtEnd_GrayAdjust_Add}
   Alpha=Power(
    Base=MaterialFunctionCall(Float2=@n20) {materialFunction=&BreakOutFloat2Components}.G
    Exp=ScalarParameter() {defaultValue=2.9963, parameterName=DAE}
   )
  )
  n22 = Saturate(None=Multiply(A=Add(A=@n3.R, B=@n21), B=ScalarParameter() {defaultValue=8.1057, group=DissolveAtEnd, parameterName=DissolveAtEnd_GrayAdjust_Mul}))
  n23 = ScalarParameter() {defaultValue=1, group=DissolveAtEnd, parameterName=DissolveAtEnd_GrayAdjust_Power}
  Multiply(A=Saturate(None=Power(Base=@n22, Exp=@n23)), B=ParticleColor().A)
  Normalize(VectorInput=VectorParameter() {parameterName=TS, defaultValue=<rgba -0.888 1 0 1>}.RGB)
  n24 = DynamicParameter() {parameterIndex=3, paramNames [X,Y,Z,A]}
  n25 = MaterialFunctionCall(X=@n24.X, Y=@n24.X, Z=@n24.X) {materialFunction=&MakeFloat3}
  n26 = ScalarParameter() {defaultValue=-1.0396, group=WPO, parameterName=WPO_GrayAdjust_Add}
  n27 = Multiply(A=Multiply(A=Add(A=@n3.R, B=@n26), B=ScalarParameter() {defaultValue=45, group=WPO, parameterName=WPO_GrayAdjust_Mul}), B=VertexNormalWS())
  n28 = Transform(None=@n27) {transformSourceType=TRANSFORMSOURCE_World, transformType=TRANSFORM_Instance}
  n29 = LocalPosition() {includedOffsets=ExcludeOffsets}
  n30 = Multiply(A=MaterialFunctionCall(Float3=Add(A=@n28, B=@n29.XYZ)) {materialFunction=&BreakOutFloat3Components}.B) {constB=-1}
  n31 = ScalarParameter() {defaultValue=529.1638, parameterName=BendRadius}
  n32 = Multiply(A=@n24.A, B=Saturate(None=Divide(A=@n30, B=@n31)))
  n33 = ParticlePositionWS()
  n34 = Add(A=@n27, B=WorldPosition() {worldPositionShaderOffset=WPT_ExcludeAllShaderOffsets}.XYZ)
  Reroute(None=Add(A=RotateAboutAxis(NormalizedRotationAxis=@n25.Result, RotationAngle=@n32, PivotPoint=@n33, Position=@n34), B=@n27))
  VectorParameter() {parameterName=RotateAxis, defaultValue=<rgba 0 1 0 0.1506>}
 ]
}
