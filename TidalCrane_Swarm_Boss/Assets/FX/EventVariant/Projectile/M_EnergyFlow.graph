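# M_EnergyFlow is a translucent, unlit material enabled for Niagara ribbons, in the event-variant projectile FX folder of the TidalCrane Swarm boss.
# Its graph lists no expressions, so no energy or flow effect is defined here.
# Automatic usage setting in the editor is off, and Niagara ribbons is the only usage flag shown.
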Material /TidalCrane_Swarm_Boss/Assets/FX/EventVariant/Projectile/M_EnergyFlow
graph {
 expressions {}
 properties {
  bAutomaticallySetUsageInEditor=false
  bUsedWithNiagaraRibbons=true
  blendMode=BLEND_Translucent
  shadingModel=MSM_Unlit
  usedShadingModels=Unlit
 }
}
