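# Translucent, unlit, two-sided liquid splash material for Niagara sprites, ribbons and mesh particles.
# A radial flowmap, scaled by dynamic CenterWarp, distorts the splash normal and mask UVs.
# Crack detail normals blend in to drive a faked GGX sky-light-direction highlight and sky reflection, added to particle color.
# Particle alpha erodes the mask, shaped by edge controls and depth fade, while MeshWPOBend scales a UV-gradient Z bend.
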Material /TidalCrane_Swarm_Boss/Assets/FX/EventVariant/Projectile/M_Liquid_Splash
graph {
 properties {
  bUsedWithNiagaraMeshParticles=true
  bUsedWithNiagaraRibbons=true
  bUsedWithNiagaraSprites=true
  blendMode=BLEND_Translucent
  pixelDepthOffsetMode=PDOM_Legacy
  shadingModel=MSM_Unlit
  twoSided=true
  usedShadingModels=Unlit
 }
 nodes [
  n1 = VectorParameter() {parameterName=Tile, defaultValue=<rgba 1 1 0 0>}
  n2 = MaterialFunctionCall(Texture Scale=Constant() {r=1.4}) {materialFunction=&ScaleUVsByCenter}
  n3 = Subtract(A=ComponentMask(None=TextureSample(UVs=@n2.UVs) {samplerType=SAMPLERTYPE_LinearColor, texture=/Game/Environments/Hexmap/HexClouds/T_RadialFlowmap01}.RGB) {g=true, r=true}) {constB=0.5}
  n4 = DynamicParameter() {parameterIndex=1, defaultValue=<rgba 0 0 1 1>, paramNames [CenterWarp,MeshWPOBend,Param3,Param4]}
  n5 = Multiply(A=AppendVector(A=@n1.R, B=@n1.R), B=Add(A=TextureCoordinate(), B=Multiply(A=@n3, B=@n4.CenterWarp)))
  n6 = TextureSampleParameter2D(UVs=@n5) {parameterName=Normal, samplerType=SAMPLERTYPE_Normal, texture=/Game/Effects/Fort_Effects/Textures/Water/T_water_splash_04_n}
  LinearInterpolate(A=@n6.RGB, B=Constant3Vector() {constant=<rgba 0 0 1 0>}, Alpha=ScalarParameter() {parameterName=Flatten Normal})
  n7 = SkyAtmosphereLightDirection()
  n8 = Multiply(A=@n7, B=Constant3Vector() {constant=<rgba -1 -1 1 0>}) {constB=-1}
  n9 = Ceil(None=DotProduct(A=CameraVectorWS(), B=@n7))
  n10 = TextureObjectParameter() {
   group=CrackN
   parameterName=CrackN_Texture
   samplerType=SAMPLERTYPE_Normal
   sortPriority=0
   texture=&/T_DetailCrackNormal
  }
  n11 = VectorParameter() {group=CrackN, parameterName=CrackN_UVControl, channelNames {a=VPan, b=UPan, g=VScale, r=UScale}, defaultValue=<rgba 2 2 0 0>}
  n12 = Transform(None=MaterialFunctionCall(
    BaseNormal=@n6.RGB
    AdditionalNormal=MaterialFunctionCall(Texture=@n10, UVControl_RGB=@n11, UVControl_A=@n11) {materialFunction=&FXCommon_TextureSample}.RGB
   ) {materialFunction=&BlendAngleCorrectedNormals}.Result)
  n13 = Saturate(None=MaterialFunctionCall!(
    LightDirctionIn=LinearInterpolate(A=@n8, B=@n7, Alpha=@n9)
    CameraVectorIn
    Roughness=ScalarParameter() {defaultValue=0.25, parameterName=R}
    NormalIn=@n12
   ) {materialFunction=/Game/Effects/Niagara/MasterMaterials/Functions/Common/FXCommon_GGX}.Result)
  n14 = ScalarParameter() {group=HighLight, parameterName=HighLight_GrayAdjust_Add}
  n15 = Saturate(None=Multiply(A=Add(A=@n13, B=@n14), B=ScalarParameter() {defaultValue=1, group=HighLight, parameterName=HighLight_GrayAdjust_Mul}))
  n16 = ScalarParameter() {defaultValue=1, group=HighLight, parameterName=HighLight_GrayAdjust_Power}
  n17 = Multiply(A=Saturate(None=Power(Base=@n15, Exp=@n16)), B=VectorParameter() {parameterName=RefC, defaultValue=<rgba 0.2 0.2 0.2 1>}.RGB)
  n18 = Multiply(
   A=SkyLightEnvMapSample(Direction=ReflectionVectorWS(CustomWorldNormal=@n12), Roughness=ScalarParameter() {defaultValue=0.25, parameterName=R2})
   B=ScalarParameter() {defaultValue=0.02, parameterName=S}
  )
  n19 = ParticleColor()
  MaterialFunctionCall(Emissive=Add(A=Add(A=@n17, B=@n18), B=@n19.RGB)) {materialFunction=&MF_FX_EmissiveAdjustments}
  n20 = TextureSampleParameter2D(UVs=@n5) {parameterName=Mask, texture=/Game/Effects/Fort_Effects/Textures/Water/T_water_splash_04}
  n21 = ChannelMaskParameter(None=@n20.RGB) {maskChannel=Blue, parameterName=ErodeMask, defaultValue=<rgba 0 0 1 0>}
  n22 = OneMinus(None=@n19.A)
  n23 = VectorParameter() {parameterName=DissolveEdge, defaultValue=<rgba 5 2 0 1>}
  n24 = Saturate(None=Multiply(A=Subtract(A=@n21.R, B=@n22), B=@n23.R))
  n25 = Add(
   A=ChannelMaskParameter(None=@n20.RGBA) {maskChannel=Blue, parameterName=OpacityMask, defaultValue=<rgba 0 0 1 0>}.R
   B=ScalarParameter() {group=LiquidEdge, parameterName=LiquidEdge_GrayAdjust_Add}
  )
  n26 = ScalarParameter() {defaultValue=2, group=LiquidEdge, parameterName=LiquidEdge_GrayAdjust_Mul}
  n27 = Power(
   Base=Saturate(None=Multiply(A=@n25, B=@n26))
   Exp=ScalarParameter() {defaultValue=1.5, group=LiquidEdge, parameterName=LiquidEdge_GrayAdjust_Power}
  )
  n28 = ScalarParameter() {defaultValue=222, parameterName=DFFF}
  Multiply(A=Multiply(A=Power(Base=@n24, Exp=@n23.G), B=Saturate(None=@n27)), B=Power(Base=DepthFade(FadeDistance=@n28)))
  n29 = ComponentMask(None=TextureCoordinate()) {g=true}
  Multiply(A=Multiply(A=Constant3Vector() {constant=<rgba 0 0 1 0>}, B=@n4.MeshWPOBend), B=Power(Base=OneMinus(None=@n29)) {constExponent=3})
  OneMinus(None=Divide(A=Time() {bOverride_Period=true, period=2})) {desc=debug erosion}
 ]
}
