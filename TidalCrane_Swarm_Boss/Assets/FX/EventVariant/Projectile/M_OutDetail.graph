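# Additive unlit Niagara ribbon material, apparently the outer detail layer of the Tidal Crane Swarm Boss event-variant projectile.
# A panning Perlin noise distorts a panning plasma texture, which feeds two separate refine branches.
# The color branch tints by DetailColor and drives its refine multiplier toward a huge Tip value via a sharp ramp along U.
# The alpha branch multiplies in noise-warped V-edge and U tip/tail masks, then particle alpha.
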Material /TidalCrane_Swarm_Boss/Assets/FX/EventVariant/Projectile/M_OutDetail
graph {
 expressions {
  MaterialExpressionAdd_0 {
   class=MaterialExpressionAdd
   inputs [
    A {from=MaterialExpressionMaterialFunctionCall_4, output=Result}
    B {from=""}
   ]
  }
  MaterialExpressionAdd_1 {
   class=MaterialExpressionAdd
   inputs [
    A {from=MaterialExpressionTextureCoordinate_1}
    B {from=MaterialExpressionMaterialFunctionCall_5, output=NoiseValue}
   ]
  }
  MaterialExpressionAdd_2 {
   class=MaterialExpressionAdd
   inputs [
    A {from=MaterialExpressionOneMinus_0}
    B {from=MaterialExpressionScalarParameter_6}
   ]
  }
  MaterialExpressionConvert_0 {
   class=MaterialExpressionConvert
   inputs [
    None {from=MaterialExpressionVectorParameter_2, output=R}
    None {from=MaterialExpressionLinearInterpolate_0}
    None {from=MaterialExpressionVectorParameter_2, output=R}
   ]
   properties {
    nodeName=Make Float3
    convertInputs [
     {
      type=Scalar
      expressionInput {
       expression=&self:MaterialExpressionVectorParameter_2
       inputName=None
       mask=1
       maskA=0
       maskB=0
       maskG=0
       maskR=1
       outputIndex=1
      }
     }
     {
      type=Scalar
      expressionInput {
       expression=&self:MaterialExpressionLinearInterpolate_0
       inputName=None
       mask=0
       maskA=0
       maskB=0
       maskG=0
       maskR=0
       outputIndex=0
      }
     }
     {
      type=Scalar
      expressionInput {
       expression=&self:MaterialExpressionVectorParameter_2
       inputName=None
       mask=1
       maskA=0
       maskB=1
       maskG=0
       maskR=0
       outputIndex=3
      }
     }
    ]
    convertOutputs [{type=Vector3}]
   }
  }
  MaterialExpressionLinearInterpolate_0 {
   class=MaterialExpressionLinearInterpolate
   inputs [
    A {from=MaterialExpressionVectorParameter_2, output=G}
    B {from=MaterialExpressionScalarParameter_0}
    Alpha {from=MaterialExpressionSaturate_3}
   ]
  }
  MaterialExpressionMaterialFunctionCall_0 {
   class=MaterialExpressionMaterialFunctionCall
   inputs [
    Texture {from=MaterialExpressionMaterialFunctionCall_2, output=R}
    RGB {from=MaterialExpressionVectorParameter_3, output=RGB}
    A {from=MaterialExpressionVectorParameter_3, output=RGB}
   ]
   properties {materialFunction=&MF_FX_Refine}
  }
  MaterialExpressionMaterialFunctionCall_1 {
   class=MaterialExpressionMaterialFunctionCall
   inputs [
    XYZ {from=""}
    XY {from=MaterialExpressionTextureCoordinate_0}
   ]
   properties {materialFunction=&Swizzle}
  }
  MaterialExpressionMaterialFunctionCall_2 {
   class=MaterialExpressionMaterialFunctionCall
   inputs [
    Texture {from=MaterialExpressionTextureObjectParameter_0}
    AdditionalUVOffset {from=MaterialExpressionMaterialFunctionCall_5, output=NoiseValue}
    UVControl_RGB {from=MaterialExpressionVectorParameter_0}
    UVControl_A {from=MaterialExpressionVectorParameter_0}
    "[Override]TimeInput" {from=""}
    "[Override]UVInput" {from=""}
   ]
   properties {materialFunction=&FXCommon_TextureSample}
  }
  MaterialExpressionMaterialFunctionCall_3 {
   class=MaterialExpressionMaterialFunctionCall
   inputs [
    Texture {from=MaterialExpressionTextureObjectParameter_1}
    AdditionalUVOffset {from=""}
    UVControl_RGB {from=MaterialExpressionVectorParameter_4}
    UVControl_A {from=MaterialExpressionVectorParameter_4}
    "[Override]TimeInput" {from=""}
    "[Override]UVInput" {from=""}
   ]
   properties {materialFunction=&FXCommon_TextureSample}
  }
  MaterialExpressionMaterialFunctionCall_4 {
   class=MaterialExpressionMaterialFunctionCall
   inputs [
    Texture {from=MaterialExpressionMaterialFunctionCall_2, output=R}
    RGB {from=MaterialExpressionConvert_0}
    A {from=MaterialExpressionVectorParameter_2, output=A}
   ]
   properties {materialFunction=&MF_FX_Refine}
  }
  MaterialExpressionMaterialFunctionCall_5 {
   class=MaterialExpressionMaterialFunctionCall
   inputs [
    NoiseValue {from=MaterialExpressionMaterialFunctionCall_3, output=R}
    Control_RGB {from=MaterialExpressionVectorParameter_5}
    Control_A {from=MaterialExpressionVectorParameter_5}
   ]
   properties {materialFunction=/Game/Effects/Niagara/MasterMaterials/Functions/Common/FXCommon_NoiseControl}
  }
  MaterialExpressionMaterialFunctionCall_6 {
   class=MaterialExpressionMaterialFunctionCall
   inputs [Float2 {from=MaterialExpressionAdd_1}]
   properties {materialFunction=&BreakOutFloat2Components}
  }
  MaterialExpressionMaterialFunctionCall_7 {
   class=MaterialExpressionMaterialFunctionCall
   inputs [
    UorV {from=MaterialExpressionMaterialFunctionCall_6, output=G}
    EdgeControl {from=MaterialExpressionVectorParameter_8}
   ]
   properties {materialFunction=/Game/Effects/Niagara/MasterMaterials/Functions/Common/FXCommon_EdgeMask_Simple}
  }
  MaterialExpressionMaterialFunctionCall_8 {
   class=MaterialExpressionMaterialFunctionCall
   inputs [Float2 {from=MaterialExpressionMaterialFunctionCall_1, output=YX}]
   properties {materialFunction=&BreakOutFloat2Components}
  }
  MaterialExpressionMaterialFunctionCall_9 {
   class=MaterialExpressionMaterialFunctionCall
   inputs [
    UorV {from=MaterialExpressionMaterialFunctionCall_6, output=R}
    Invert_EdgeControl {from=MaterialExpressionVectorParameter_7}
    NoInvert_EdgeControl {from=MaterialExpressionVectorParameter_6}
   ]
   properties {materialFunction=/Game/Effects/Niagara/MasterMaterials/Functions/Common/FXCommon_EdgeMask_Ex}
  }
  MaterialExpressionMultiply_0 {
   class=MaterialExpressionMultiply
   inputs [
    A {from=MaterialExpressionMaterialFunctionCall_0, output=Result}
    B {from=MaterialExpressionMaterialFunctionCall_7, output=Mask}
   ]
  }
  MaterialExpressionMultiply_1 {
   class=MaterialExpressionMultiply
   inputs [
    A {from=MaterialExpressionMultiply_0}
    B {from=MaterialExpressionMaterialFunctionCall_9, output=Mask}
   ]
  }
  MaterialExpressionMultiply_10 {
   class=MaterialExpressionMultiply
   inputs [
    A {from=MaterialExpressionAdd_2}
    B {from=MaterialExpressionScalarParameter_7}
   ]
  }
  MaterialExpressionMultiply_11 {
   class=MaterialExpressionMultiply
   inputs [
    A {from=MaterialExpressionAdd_0}
    B {from=MaterialExpressionVectorParameter_1, output=RGB}
   ]
  }
  MaterialExpressionMultiply_2 {
   class=MaterialExpressionMultiply
   inputs [
    A {from=MaterialExpressionParticleColor_0, output=A}
    B {from=MaterialExpressionMultiply_1}
   ]
  }
  MaterialExpressionOneMinus_0 {class=MaterialExpressionOneMinus, inputs [None {from=MaterialExpressionMaterialFunctionCall_8, output=G}]}
  MaterialExpressionParticleColor_0 {class=MaterialExpressionParticleColor, inputs [], properties {}}
  MaterialExpressionPower_6 {
   class=MaterialExpressionPower
   inputs [
    Base {from=MaterialExpressionSaturate_2}
    Exp {from=MaterialExpressionScalarParameter_8}
   ]
  }
  MaterialExpressionSaturate_2 {class=MaterialExpressionSaturate, inputs [None {from=MaterialExpressionMultiply_10}]}
  MaterialExpressionSaturate_3 {class=MaterialExpressionSaturate, inputs [None {from=MaterialExpressionPower_6}]}
  MaterialExpressionScalarParameter_0 {class=MaterialExpressionScalarParameter, properties {defaultValue=53999.9531, parameterName=Tip}}
  MaterialExpressionScalarParameter_6 {class=MaterialExpressionScalarParameter, properties {defaultValue=-0.16, group=TipScaleUp, parameterName=TipScaleUp_GrayAdjust_Add}}
  MaterialExpressionScalarParameter_7 {class=MaterialExpressionScalarParameter, properties {defaultValue=1.2809, group=TipScaleUp, parameterName=TipScaleUp_GrayAdjust_Mul}}
  MaterialExpressionScalarParameter_8 {class=MaterialExpressionScalarParameter, properties {defaultValue=43.5067, group=TipScaleUp, parameterName=TipScaleUp_GrayAdjust_Power}}
  MaterialExpressionTextureCoordinate_0 {class=MaterialExpressionTextureCoordinate}
  MaterialExpressionTextureCoordinate_1 {class=MaterialExpressionTextureCoordinate}
  MaterialExpressionTextureObjectParameter_0 {
   class=MaterialExpressionTextureObjectParameter
   inputs [
    UVs {from="", output=""}
    Apply View MipBias {from="", output=""}
   ]
   properties {
    group=DetailTexture
    parameterName=DetailTexture_Texture
    samplerType=SAMPLERTYPE_LinearGrayscale
    sortPriority=0
    texture=/Game/Effects/Fort_Effects/Emotes/UltraLight/Texture/T_Plasma_B_Greyscale
   }
  }
  MaterialExpressionTextureObjectParameter_1 {
   class=MaterialExpressionTextureObjectParameter
   inputs [
    UVs {from="", output=""}
    Apply View MipBias {from="", output=""}
   ]
   properties {
    group=Noise
    parameterName=Noise_Texture
    samplerType=SAMPLERTYPE_LinearGrayscale
    sortPriority=0
    texture=&T_SimplePerlin_Greyscale
   }
  }
  MaterialExpressionVectorParameter_0 {
   class=MaterialExpressionVectorParameter
   properties {
    group=DetailTexture
    parameterName=DetailTexture_UVControl
    channelNames {a=VPan, b=UPan, g=VScale, r=UScale}
    defaultValue=<rgba 0.3 0.5 -0.7 0>
   }
  }
  MaterialExpressionVectorParameter_1 {class=MaterialExpressionVectorParameter, properties {parameterName=DetailColor, defaultValue=<rgba 31.6667 0 50 1>}}
  MaterialExpressionVectorParameter_2 {class=MaterialExpressionVectorParameter, properties {parameterName=DetailTexture_Refine, defaultValue=<rgba 7 6 0 0.5>}}
  MaterialExpressionVectorParameter_3 {class=MaterialExpressionVectorParameter, properties {parameterName=DetailTexture_RefineO, defaultValue=<rgba 8 2 0.015 0.5>}}
  MaterialExpressionVectorParameter_4 {
   class=MaterialExpressionVectorParameter
   properties {group=Noise, parameterName=Noise_UVControl, channelNames {a=VPan, b=UPan, g=VScale, r=UScale}, defaultValue=<rgba 0.5 0.3 -1 0>}
  }
  MaterialExpressionVectorParameter_5 {
   class=MaterialExpressionVectorParameter
   properties {
    group=Noise
    parameterName=Noise_NoiseControl
    sortPriority=60
    channelNames {a=VStrength, b=UStrength, g=VOffset, r=UOffset}
    defaultValue=<rgba -0.5 -0.5 0 0.2>
   }
  }
  MaterialExpressionVectorParameter_6 {
   class=MaterialExpressionVectorParameter
   properties {
    group=TipTaill
    parameterName=TipTaill_EdgeMask_Control
    channelNames {a="", b=Power, g=Multiply, r=Offset}
    defaultValue=<rgba 0 10 2.0419 1>
   }
  }
  MaterialExpressionVectorParameter_7 {
   class=MaterialExpressionVectorParameter
   properties {
    group=TipTaill
    parameterName=TipTaill_EdgeMask_InvertControl
    channelNames {a="", b=Power, g=Multiply, r=Offset}
    defaultValue=<rgba 0 1.5 1.5 1>
   }
  }
  MaterialExpressionVectorParameter_8 {
   class=MaterialExpressionVectorParameter
   properties {group=Noise, parameterName=Noise_EdgeMask_Control, channelNames {a="", b=Power, g=Multiply, r=Offset}, defaultValue=<rgba 0 2.6209 2 1>}
  }
 }
 properties {
  bAutomaticallySetUsageInEditor=false
  bUsedWithNiagaraRibbons=true
  blendMode=BLEND_Additive
  shadingModel=MSM_Unlit
  usedShadingModels=Unlit
 }
}
