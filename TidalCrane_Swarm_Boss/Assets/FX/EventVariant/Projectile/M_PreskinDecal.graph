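# Unlit translucent Niagara mesh-particle material that draws a decal-like shell effect with purple glow, mapped by world-space-aligned screen coordinates.
# Shell, rolling plasma and dust textures form the glow, adding GGX highlights and sky-light reflections from blended shell and crack normals to fake lighting.
# Particle color tints it; opacity combines particle alpha, dynamic-parameter noise dissolve, radial mask, scene-depth cutoff and depth fade.
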Material /TidalCrane_Swarm_Boss/Assets/FX/EventVariant/Projectile/M_PreskinDecal
graph {
 properties {
  bAutomaticallySetUsageInEditor=false
  bUsedWithNiagaraMeshParticles=true
  blendMode=BLEND_Translucent
  shadingModel=MSM_Unlit
  usedShadingModels=Unlit
 }
 nodes [
  n1 = ParticleColor()
  n2 = TextureObjectParameter() {
   group=DTexture
   parameterName=DTexture_Texture
   sortPriority=0
   texture=&/T_ShellTexture
  }
  n3 = MaterialFunctionCall!(Center Point World Position) {materialFunction=/Engine/Functions/Engine_MaterialFunctions02/UVs/WorldSpaceAlignedScreenCoordinates}
  n4 = VectorParameter() {group=DTexture, parameterName=DTexture_UVControl, channelNames {a=VPan, b=UPan, g=VScale, r=UScale}, defaultValue=<rgba 2.3 2.3 0 0>}
  n5 = MaterialFunctionCall(Texture=@n2, AdditionalUVOffset=@n3.Offset and Scaled UVS, UVControl_RGB=@n4, UVControl_A=@n4, "[Override]UVInput"=@n3.Offset and Scaled UVS) {materialFunction=&FXCommon_TextureSample}
  n6 = Saturate(None=@n5.R)
  n7 = VectorParameter() {parameterName=DTexture_Refine, defaultValue=<rgba 20 13 1 0.5>}
  n8 = MaterialFunctionCall(X=@n7.R, Y=@n7.R, Z=@n7.R) {materialFunction=&MakeFloat3}
  n9 = TextureObjectParameter() {
   group=RollingT
   parameterName=RollingT_Texture
   samplerType=SAMPLERTYPE_LinearColor
   sortPriority=0
   texture=/Game/Items/Traps/Textures/T_Plasma_Flames_02_M
  }
  n10 = VectorParameter() {group=RollingT, parameterName=RollingT_UVControl, channelNames {a=VPan, b=UPan, g=VScale, r=UScale}, defaultValue=<rgba 1 1 0.1 0.15>}
  n11 = VectorParameter() {parameterName=RollingT_Refine, defaultValue=<rgba 0.8336 1 1 0.5>}
  n12 = MaterialFunctionCall(
   Texture=MaterialFunctionCall(Texture=@n9, UVControl_RGB=@n10, UVControl_A=@n10, "[Override]UVInput"=@n3.Offset and Scaled UVS) {materialFunction=&FXCommon_TextureSample}.R
   RGB=@n11.RGB
   A=@n11.RGB
  ) {materialFunction=&MF_FX_Refine}
  n13 = TextureObjectParameter() {group=Detail, parameterName=Detail_Texture, sortPriority=0, texture=/Game/Global/Textures/Default/PointSplatter/T_Tiling_Dust}
  n14 = VectorParameter() {group=Detail, parameterName=Detail_UVControl, channelNames {a=VPan, b=UPan, g=VScale, r=UScale}, defaultValue=<rgba 1 1 0 0>}
  n15 = MaterialFunctionCall(Texture=@n13, UVControl_RGB=@n14, UVControl_A=@n14, "[Override]UVInput"=@n3.Offset and Scaled UVS) {materialFunction=&FXCommon_TextureSample}
  n16 = VectorParameter() {parameterName=Detail_Refine, defaultValue=<rgba 3.5 483.7693 1 0.3>}
  n17 = Add(
   A=Multiply(
    A=MaterialFunctionCall(Texture=@n15.R, RGB=@n16.RGB, A=@n16.RGB) {materialFunction=&MF_FX_Refine}.Result
    B=VectorParameter() {parameterName=DtailTexColor, defaultValue=<rgba 48.347 0 50 1>}.RGB
   )
   B=VectorParameter() {parameterName=DTexColor, defaultValue=<rgba 2.6667 0 4 1>}.RGB
  )
  n18 = MaterialFunctionCall(Float3=TextureSample(UVs=@n5.UV) {samplerType=SAMPLERTYPE_Normal, texture=&/T_ShellTextureNormal}.RGB) {materialFunction=&BreakOutFloat3Components}
  n19 = Multiply(A=@n18.R) {constB=-1}
  n20 = Multiply(A=@n18.G) {constB=-1}
  n21 = TextureObjectParameter() {
   group=CrackN
   parameterName=CrackN_Texture
   samplerType=SAMPLERTYPE_Normal
   sortPriority=0
   texture=&/T_DetailCrackNormal
  }
  n22 = VectorParameter() {group=CrackN, parameterName=CrackN_UVControl, channelNames {a=VPan, b=UPan, g=VScale, r=UScale}, defaultValue=<rgba 2.5 2.5 0 0>}
  n23 = MaterialFunctionCall(Texture=@n21, UVControl_RGB=@n22, UVControl_A=@n22, "[Override]UVInput"=@n3.Offset and Scaled UVS) {materialFunction=&FXCommon_TextureSample}
  n24 = ScalarParameter() {defaultValue=-1, parameterName=F}
  n25 = Transform(None=MaterialFunctionCall(
    BaseNormal=MaterialFunctionCall(X=@n19, Y=@n20, Z=@n18.B) {materialFunction=&MakeFloat3}.Result
    AdditionalNormal=MaterialFunctionCall(Normal=@n23.RGB, Flatness=@n24) {materialFunction=&FlattenNormal}.Result
   ) {materialFunction=&BlendAngleCorrectedNormals}.Result)
  n26 = MaterialFunctionCall!(LightDirctionIn, CameraVectorIn, Roughness=ScalarParameter() {defaultValue=0.4, parameterName=R}, NormalIn=@n25) {materialFunction=/Game/Effects/Niagara/MasterMaterials/Functions/Common/FXCommon_GGX}
  n27 = Multiply(
   A=Add(A=Saturate(None=@n26.Result), B=ScalarParameter() {group=HighLight, parameterName=HighLight_GrayAdjust_Add})
   B=ScalarParameter() {defaultValue=1, group=HighLight, parameterName=HighLight_GrayAdjust_Mul}
  )
  n28 = Saturate(None=Power(Base=Saturate(None=@n27), Exp=ScalarParameter() {defaultValue=1, group=HighLight, parameterName=HighLight_GrayAdjust_Power}))
  n29 = VectorParameter() {parameterName=RefC, defaultValue=<rgba 0.33 0.33 0.33 1>}
  n30 = SkyLightEnvMapSample(Direction=ReflectionVectorWS(CustomWorldNormal=@n25), Roughness=ScalarParameter() {defaultValue=0.15, parameterName=R2})
  n31 = ScalarParameter() {defaultValue=0.04, parameterName=S}
  n32 = Add(
   A=Multiply(
    A=MaterialFunctionCall(Texture=@n6, RGB=@n8.Result, A=@n7.A) {materialFunction=&MF_FX_Refine}.Result
    B=Multiply(A=@n12.Result, B=@n17)
   )
   B=Add(A=Multiply(A=@n28, B=@n29.RGB), B=Multiply(A=@n30, B=@n31))
  )
  MaterialFunctionCall(Emissive=Multiply(A=@n1.RGB, B=Divide(A=@n32))) {materialFunction=&MF_FX_EmissiveAdjustments}
  n33 = TextureObjectParameter() {
   group=Dissolve
   parameterName=Dissolve_Texture
   samplerType=SAMPLERTYPE_LinearColor
   sortPriority=0
   texture=/Engine/Functions/Engine_MaterialFunctions02/ExampleContent/Textures/OffsetNoiseDistanceFields_ForNormals
  }
  n34 = VectorParameter() {group=Dissolve, parameterName=Dissolve_UVControl, channelNames {a=VPan, b=UPan, g=VScale, r=UScale}, defaultValue=<rgba 5 5 0 0>}
  n35 = MaterialFunctionCall(Texture=@n33, UVControl_RGB=@n34, UVControl_A=@n34, "[Override]UVInput"=@n3.Offset and Scaled UVS) {materialFunction=&FXCommon_TextureSample}
  n36 = Add(
   A=Subtract(A=OneMinus(None=@n35.B), B=DynamicParameter() {paramNames [Dissolve,Param2,Param3,Param4]}.Dissolve)
   B=ScalarParameter() {group=Dissolve, parameterName=Dissolve_GrayAdjust_Add}
  )
  n37 = ScalarParameter() {defaultValue=3.4088, group=Dissolve, parameterName=Dissolve_GrayAdjust_Mul}
  n38 = Power(
   Base=Saturate(None=Multiply(A=@n36, B=@n37))
   Exp=ScalarParameter() {defaultValue=1, group=Dissolve, parameterName=Dissolve_GrayAdjust_Power}
  )
  n39 = MaterialFunctionCall(Vector or UVs=@n3.Offset and Scaled UVS) {materialFunction=&VectorToRadialValue}
  n40 = ScalarParameter() {defaultValue=0.3076, parameterName=T}
  n41 = OneMinus(None=MaterialFunctionCall(Float2=Divide(A=@n39.Radial Coordinates, B=@n40)) {materialFunction=&BreakOutFloat2Components}.G)
  n42 = ScalarParameter() {group=Mask, parameterName=Mask_GrayAdjust_Add}
  n43 = Saturate(None=Multiply(A=Add(A=@n41, B=@n42), B=ScalarParameter() {defaultValue=1.3774, group=Mask, parameterName=Mask_GrayAdjust_Mul}))
  n44 = ScalarParameter() {defaultValue=3.5157, group=Mask, parameterName=Mask_GrayAdjust_Power}
  n45 = SceneTexture() {bFiltered=true, sceneTextureId=PPI_SceneDepth}
  n46 = Saturate(None=Divide(
    A=MaterialFunctionCall(Float3=@n45.Color) {materialFunction=&BreakOutFloat3Components}.R
    B=ScalarParameter() {defaultValue=5000, parameterName=CustomDepthDivide}
   ) {constB=4096})
  n47 = Multiply(A=Saturate(None=Power(Base=@n43, Exp=@n44)), B=OneMinus(None=Floor(None=@n46)))
  n48 = Power(Base=DepthFade(FadeDistance=ScalarParameter() {defaultValue=100, parameterName=DF}))
  Multiply(A=@n1.A, B=Multiply(A=Saturate(None=@n38), B=Multiply(A=@n47, B=@n48)))
 ]
}
