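# Translucent unlit material for Niagara mesh particles, shading a cracked shell projectile in the Tidal Crane swarm boss event-variant FX.
# Colour fakes lighting with mask-tinted base, GGX highlight and skylight reflection, using blended shell and crack normals.
# Opacity combines shell mask, noise-distorted V gradients for tip dissolve and end fades, and particle alpha.
# World position offset displaces along normals by mask and bends the mesh about particle position, driven by a dynamic parameter.
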Material /TidalCrane_Swarm_Boss/Assets/FX/EventVariant/Projectile/M_ShellMat
graph {
 properties {
  bAutomaticallySetUsageInEditor=false
  bUsedWithNiagaraMeshParticles=true
  blendMode=BLEND_Translucent
  shadingModel=MSM_Unlit
  usedShadingModels=Unlit
 }
 nodes [
  ActorPositionWS()
  n1 = TextureObjectParameter() {
   group=ShellMask
   parameterName=ShellMask_Texture
   sortPriority=0
   texture=&/T_ShellTexture
  }
  n2 = VectorParameter() {group=ShellMask, parameterName=ShellMask_UVControl, channelNames {a=VPan, b=UPan, g=VScale, r=UScale}, defaultValue=<rgba 1 0.5 0 0.8>}
  n3 = MaterialFunctionCall(Texture=@n1, UVControl_RGB=@n2, UVControl_A=@n2) {materialFunction=&FXCommon_TextureSample}
  n4 = OneMinus(None=@n3.R)
  n5 = VectorParameter() {parameterName=BaseC_Refine, defaultValue=<rgba 4.6023 4.0905 1 0.5>}
  n6 = Add(
   A=Multiply(
    A=VectorParameter() {parameterName=BaseC, defaultValue=<rgba 0.0433 0 0.0521 1>}.RGB
    B=MaterialFunctionCall(Texture=@n4, RGB=@n5.RGB, A=@n5.RGB) {materialFunction=&MF_FX_Refine}.Result
   )
   B=VectorParameter() {defaultValue=<rgba 0.0208 0.0208 0.0208 1>}.RGB
  )
  n7 = TextureObjectParameter() {
   group=CrackN
   parameterName=CrackN_Texture
   samplerType=SAMPLERTYPE_Normal
   sortPriority=0
   texture=&/T_DetailCrackNormal
  }
  n8 = VectorParameter() {group=CrackN, parameterName=CrackN_UVControl, channelNames {a=VPan, b=UPan, g=VScale, r=UScale}, defaultValue=<rgba 3 3 0 0>}
  n9 = Transform(None=MaterialFunctionCall(
    BaseNormal=TextureSample(UVs=@n3.UV) {samplerType=SAMPLERTYPE_Normal, texture=&/T_ShellTextureNormal}.RGB
    AdditionalNormal=MaterialFunctionCall(Texture=@n7, UVControl_RGB=@n8, UVControl_A=@n8, "[Override]UVInput"=@n3.UV) {materialFunction=&FXCommon_TextureSample}.RGB
   ) {materialFunction=&BlendAngleCorrectedNormals}.Result)
  n10 = Saturate(None=MaterialFunctionCall!(LightDirctionIn, CameraVectorIn, Roughness=ScalarParameter() {defaultValue=0.6232, parameterName=R}, NormalIn=@n9) {materialFunction=/Game/Effects/Niagara/MasterMaterials/Functions/Common/FXCommon_GGX}.Result)
  n11 = ScalarParameter() {group=HighLight, parameterName=HighLight_GrayAdjust_Add}
  n12 = Saturate(None=Multiply(A=Add(A=@n10, B=@n11), B=ScalarParameter() {defaultValue=1, group=HighLight, parameterName=HighLight_GrayAdjust_Mul}))
  n13 = ScalarParameter() {defaultValue=1, group=HighLight, parameterName=HighLight_GrayAdjust_Power}
  n14 = Multiply(A=Saturate(None=Power(Base=@n12, Exp=@n13)), B=VectorParameter() {parameterName=RefC, defaultValue=<rgba 0.33 0.33 0.33 1>}.RGB)
  n15 = TextureObjectParameter() {
   group=DissolveNoise
   parameterName=DissolveNoise_Texture
   samplerType=SAMPLERTYPE_LinearGrayscale
   sortPriority=0
   texture=/Game/Effects/Fort_Effects/Emotes/UltraLight/Texture/T_Plasma_A_Greyscale
  }
  n16 = VectorParameter() {
   group=DissolveNoise
   parameterName=DissolveNoise_UVControl
   channelNames {a=VPan, b=UPan, g=VScale, r=UScale}
   defaultValue=<rgba 0.5 0.5 0 0.608>
  }
  n17 = MaterialFunctionCall(Texture=@n15, UVControl_RGB=@n16, UVControl_A=@n16) {materialFunction=&FXCommon_TextureSample}
  n18 = VectorParameter() {
   group=DissolveNoise
   parameterName=DissolveNoise_NoiseControl
   sortPriority=60
   channelNames {a=VStrength, b=UStrength, g=VOffset, r=UOffset}
   defaultValue=<rgba -0.5 -0.5 0 0.464>
  }
  n19 = MaterialFunctionCall(Float2=Add(
    A=TextureCoordinate()
    B=MaterialFunctionCall!(NoiseValue=@n17.R, Control_RGB=@n18, Control_A=@n18) {materialFunction=/Game/Effects/Niagara/MasterMaterials/Functions/Common/FXCommon_NoiseControl}.NoiseValue
   )) {materialFunction=&BreakOutFloat2Components}
  n20 = ScalarParameter() {defaultValue=-0.7648, group=TipDissolve, parameterName=TipDissolve_GrayAdjust_Add}
  n21 = Saturate(None=Multiply(A=Add(A=@n19.G, B=@n20), B=ScalarParameter() {defaultValue=1.6823, group=TipDissolve, parameterName=TipDissolve_GrayAdjust_Mul}))
  n22 = ScalarParameter() {defaultValue=1, group=TipDissolve, parameterName=TipDissolve_GrayAdjust_Power}
  n23 = Subtract(A=@n3.R, B=Saturate(None=Power(Base=@n21, Exp=@n22)))
  n24 = VectorParameter() {parameterName=ShellMask_Refine, defaultValue=<rgba 3 4128716.5 1 0.5>}
  n25 = MaterialFunctionCall(Texture=Saturate(None=@n23), RGB=@n24.RGB, A=@n24.RGB) {materialFunction=&MF_FX_Refine}
  n26 = ReflectionVectorWS(CustomWorldNormal=@n9)
  n27 = ScalarParameter() {defaultValue=0.15, parameterName=R2}
  Add(
   A=LinearInterpolate(A=VectorParameter() {parameterName=S, defaultValue=<rgba 0.137 0.045 0.15 1>}.RGB, B=Add(A=@n6, B=@n14), Alpha=Saturate(None=@n25.Result))
   B=Multiply(A=SkyLightEnvMapSample(Direction=@n26, Roughness=@n27), B=ScalarParameter() {defaultValue=0.03, parameterName=S})
  )
  Constant3Vector()
  n28 = VectorParameter() {parameterName=ToOpa2_Refine, defaultValue=<rgba 0.5773 12.0157 0 0>}
  n29 = Saturate(None=MaterialFunctionCall(Texture=@n3.R, RGB=@n28.RGB, A=@n28.RGB) {materialFunction=&MF_FX_Refine}.Result)
  n30 = ScalarParameter() {defaultValue=0.2, parameterName=Opa2}
  n31 = Saturate(None=Add(A=Multiply(A=@n29, B=@n30), B=@n25.Result))
  n32 = Multiply(
   A=Add(A=@n19.G, B=ScalarParameter() {defaultValue=0.15, group=TipF2, parameterName=TipF2_GrayAdjust_Add})
   B=ScalarParameter() {defaultValue=1, group=TipF2, parameterName=TipF2_GrayAdjust_Mul}
  )
  n33 = Saturate(None=Power(Base=Saturate(None=@n32), Exp=ScalarParameter() {defaultValue=34.3286, group=TipF2, parameterName=TipF2_GrayAdjust_Power}))
  n34 = Multiply(
   A=Add(A=OneMinus(None=@n19.G), B=ScalarParameter() {defaultValue=0.1719, group=TailFade, parameterName=TailFade_GrayAdjust_Add})
   B=ScalarParameter() {defaultValue=1, group=TailFade, parameterName=TailFade_GrayAdjust_Mul}
  )
  n35 = Saturate(None=Power(Base=Saturate(None=@n34), Exp=ScalarParameter() {defaultValue=10.8311, group=TailFade, parameterName=TailFade_GrayAdjust_Power}))
  Multiply(A=LinearInterpolate(A=LinearInterpolate(A=@n31, Alpha=@n33) {constB=0}, Alpha=@n35) {constB=0}, B=ParticleColor().A)
  n36 = DynamicParameter() {parameterIndex=3, paramNames [X,Y,Z,A]}
  n37 = MaterialFunctionCall(X=@n36.X, Y=@n36.X, Z=@n36.X) {materialFunction=&MakeFloat3}
  n38 = ScalarParameter() {defaultValue=-0.2227, group=WPO, parameterName=WPO_GrayAdjust_Add}
  n39 = Multiply(A=Multiply(A=Add(A=@n3.R, B=@n38), B=ScalarParameter() {defaultValue=25, group=WPO, parameterName=WPO_GrayAdjust_Mul}), B=VertexNormalWS())
  n40 = Transform(None=@n39) {transformSourceType=TRANSFORMSOURCE_World, transformType=TRANSFORM_Instance}
  n41 = LocalPosition() {includedOffsets=ExcludeOffsets}
  n42 = Multiply(A=MaterialFunctionCall(Float3=Add(A=@n40, B=@n41.XYZ)) {materialFunction=&BreakOutFloat3Components}.B) {constB=-1}
  n43 = ScalarParameter() {defaultValue=529.1638, parameterName=BendRadius}
  n44 = Multiply(A=@n36.A, B=Saturate(None=Divide(A=@n42, B=@n43)))
  n45 = ParticlePositionWS()
  n46 = Add(A=@n39, B=WorldPosition() {worldPositionShaderOffset=WPT_ExcludeAllShaderOffsets}.XYZ)
  Reroute(None=Add(A=RotateAboutAxis(NormalizedRotationAxis=@n37.Result, RotationAngle=@n44, PivotPoint=@n45, Position=@n46), B=@n39))
  Subtract(A=@n23, B=@n35)
  VectorParameter() {parameterName=RotateAxis, defaultValue=<rgba 0 1 0 0.1506>}
 ]
}
