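# Translucent unlit Niagara ribbon material for a shell-textured projectile trail, in the TidalCrane swarm boss event-variant FX.
# Fakes lit color using a normal-mapped GGX highlight, skylight environment reflection and a tinted, inverted shell texture.
# Builds an opacity mask by subtracting noise-distorted tail and side fades plus particle alpha from the shell texture, then sharpening it.
# A tip/tail edge mask and the dynamic OpacityScale parameter scale that mask.
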Material /TidalCrane_Swarm_Boss/Assets/FX/EventVariant/Projectile/M_ShellRibbon
graph {
 properties {
  bAutomaticallySetUsageInEditor=false
  bUsedWithNiagaraRibbons=true
  blendMode=BLEND_Translucent
  shadingModel=MSM_Unlit
  usedShadingModels=Unlit
 }
 nodes [
  n1 = TextureObjectParameter() {
   group=MainTexture
   parameterName=MainTexture_Texture
   sortPriority=0
   texture=&/T_ShellTexture
  }
  n2 = VectorParameter() {
   group=MainTexture
   parameterName=MainTexture_UVControl
   channelNames {a=VPan, b=UPan, g=VScale, r=UScale}
   defaultValue=<rgba 1.5 1 -0.392 0>
  }
  n3 = MaterialFunctionCall(Texture=@n1, UVControl_RGB=@n2, UVControl_A=@n2) {materialFunction=&FXCommon_TextureSample}
  n4 = Transform(None=TextureSample(UVs=@n3.UV) {samplerType=SAMPLERTYPE_Normal, texture=&/T_ShellTextureNormal}.RGB)
  n5 = MaterialFunctionCall!(LightDirctionIn, CameraVectorIn, Roughness=ScalarParameter() {defaultValue=0.6232, parameterName=R}, NormalIn=@n4) {materialFunction=/Game/Effects/Niagara/MasterMaterials/Functions/Common/FXCommon_GGX}
  n6 = Multiply(
   A=Add(A=Saturate(None=@n5.Result), B=ScalarParameter() {group=HighLight, parameterName=HighLight_GrayAdjust_Add})
   B=ScalarParameter() {defaultValue=1, group=HighLight, parameterName=HighLight_GrayAdjust_Mul}
  )
  n7 = Saturate(None=Power(Base=Saturate(None=@n6), Exp=ScalarParameter() {defaultValue=1, group=HighLight, parameterName=HighLight_GrayAdjust_Power}))
  n8 = VectorParameter() {parameterName=RefC, defaultValue=<rgba 0.33 0.33 0.33 1>}
  n9 = SkyLightEnvMapSample(Direction=ReflectionVectorWS(CustomWorldNormal=@n4), Roughness=ScalarParameter() {defaultValue=0.15, parameterName=R2})
  n10 = ScalarParameter() {defaultValue=0.03, parameterName=S}
  n11 = VectorParameter() {parameterName=BaseC, defaultValue=<rgba 0.0433 0 0.0521 1>}
  n12 = VectorParameter() {parameterName=BaseC_Refine, defaultValue=<rgba 4.6023 4.0905 1 0.5>}
  n13 = MaterialFunctionCall(Texture=OneMinus(None=@n3.R), RGB=@n12.RGB, A=@n12.RGB) {materialFunction=&MF_FX_Refine}
  Add(
   A=Add(A=Multiply(A=@n7, B=@n8.RGB), B=Multiply(A=@n9, B=@n10))
   B=Add(A=Multiply(A=@n11.RGB, B=@n13.Result), B=VectorParameter() {parameterName=Param5, defaultValue=<rgba 0.02 0.02 0.02 0>}.RGB)
  )
  n14 = VectorParameter() {defaultValue=<rgba 0 1 1 0>}
  n15 = TextureObjectParameter() {
   group=DissolveNoise
   parameterName=DissolveNoise_Texture
   samplerType=SAMPLERTYPE_LinearGrayscale
   sortPriority=0
   texture=/Game/Effects/Fort_Effects/Emotes/UltraLight/Texture/T_Plasma_A_Greyscale
  }
  n16 = VectorParameter() {
   group=DissolveNoise
   parameterName=DissolveNoise_UVControl
   channelNames {a=VPan, b=UPan, g=VScale, r=UScale}
   defaultValue=<rgba 0.25 0 -1.0982 0>
  }
  n17 = MaterialFunctionCall(Texture=@n15, UVControl_RGB=@n16, UVControl_A=@n16) {materialFunction=&FXCommon_TextureSample}
  n18 = VectorParameter() {
   group=DissolveNoise
   parameterName=DissolveNoise_NoiseControl
   sortPriority=60
   channelNames {a=VStrength, b=UStrength, g=VOffset, r=UOffset}
   defaultValue=<rgba -0.5 -0.5 0 0>
  }
  n19 = MaterialFunctionCall(Float2=Add(
    A=TextureCoordinate()
    B=MaterialFunctionCall!(NoiseValue=@n17.R, Control_RGB=@n18, Control_A=@n18) {materialFunction=/Game/Effects/Niagara/MasterMaterials/Functions/Common/FXCommon_NoiseControl}.NoiseValue
   )) {materialFunction=&BreakOutFloat2Components}
  n20 = VectorParameter() {group=EdgeFade, parameterName=EdgeFade_EdgeMask_Control, channelNames {a="", b=Power, g=Multiply, r=Offset}, defaultValue=<rgba 0 3 2 1>}
  n21 = VectorParameter() {parameterName=Se, defaultValue=<rgba 6 2.3403 0 1>}
  n22 = Multiply(
   A=Add(A=@n19.R, B=ScalarParameter() {defaultValue=-0.152, group=LerpTheSideDissolve, parameterName=LerpTheSideDissolve_GrayAdjust_Add})
   B=ScalarParameter() {defaultValue=2.517, group=LerpTheSideDissolve, parameterName=LerpTheSideDissolve_GrayAdjust_Mul}
  )
  n23 = Saturate(None=Power(
    Base=Saturate(None=@n22)
    Exp=ScalarParameter() {defaultValue=0.5435, group=LerpTheSideDissolve, parameterName=LerpTheSideDissolve_GrayAdjust_Power}
   ))
  n24 = MaterialFunctionCall!(
   UorV=@n19.G
   EdgeControl=MaterialFunctionCall(X=@n20.Offset, Y=LinearInterpolate(A=@n21.R, B=@n21.R, Alpha=@n23), Z=@n20.Offset) {materialFunction=&MakeFloat3}.Result
  ) {materialFunction=/Game/Effects/Niagara/MasterMaterials/Functions/Common/FXCommon_EdgeMask_Simple}
  LinearInterpolate(A=@n14.R, B=@n14.R, Alpha=@n24.Mask)
  n25 = VectorParameter() {
   group=TipTailOpacityFade
   parameterName=TipTailOpacityFade_EdgeMask_InvertControl
   channelNames {a="", b=Power, g=Multiply, r=Offset}
   defaultValue=<rgba 0 2.0213 3.7883 1>
  }
  n26 = VectorParameter() {
   group=TipTailOpacityFade
   parameterName=TipTailOpacityFade_EdgeMask_Control
   channelNames {a="", b=Power, g=Multiply, r=Offset}
   defaultValue=<rgba 0 20.8871 2 1>
  }
  n27 = Add(A=@n19.R, B=ScalarParameter() {defaultValue=-0.304, group=TailFade, parameterName=TailFade_GrayAdjust_Add})
  n28 = ScalarParameter() {defaultValue=1.5, group=TailFade, parameterName=TailFade_GrayAdjust_Mul}
  n29 = Power(
   Base=Saturate(None=Multiply(A=@n27, B=@n28))
   Exp=ScalarParameter() {defaultValue=1, group=TailFade, parameterName=TailFade_GrayAdjust_Power}
  )
  n30 = Subtract(A=@n3.R, B=LinearInterpolate(A=@n14.R, B=@n14.R, Alpha=Saturate(None=@n29)))
  n31 = LinearInterpolate(A=@n14.B, B=@n14.B, Alpha=@n24.Mask)
  n32 = VectorParameter() {parameterName=ShellMask_Refine, defaultValue=<rgba 1.8072 507.0689 1 0.5>}
  n33 = MaterialFunctionCall(Texture=Subtract(A=Subtract(A=@n30, B=@n31), B=ParticleColor().A), RGB=@n32.RGB, A=@n32.RGB) {materialFunction=&MF_FX_Refine}
  Multiply(
   A=Multiply(
    A=MaterialFunctionCall!(UorV=@n19.R, Invert_EdgeControl=@n25, NoInvert_EdgeControl=@n26) {materialFunction=/Game/Effects/Niagara/MasterMaterials/Functions/Common/FXCommon_EdgeMask_Ex}.Mask
    B=Saturate(None=@n33.Result)
   )
   B=DynamicParameter() {paramNames [OpacityScale,Param2,Param3,Param4]}.OpacityScale
  )
 ]
}
