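# Niagara projectile effect for the EventQueen variant, layering dark-energy meshes, ribbon trails, center smoke and GPU dissolve particles.
# A system CalculateBend scratch module feeds bend axis and angle to the wave mesh material, while Timeline cross-fades two distance-spawned trail emitters.
# Minimal001 draws three ribbons from one emitter through extra color and width attributes, matching the disabled Minimal002/003.
# Camera-distance functions drive fades on waves, smoke and trails.
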NiagaraSystem /TidalCrane_Swarm_Boss/Assets/FX/EventVariant/Projectile/NS_Projectile_EventQueen
dependencies {
 usedDataInterfaces [
  &NiagaraDataInterfaceCamera
  &NiagaraDataInterfaceCurve
  &NiagaraDataInterfaceCollisionQuery
  &NiagaraDataInterfaceLandscape
  &NiagaraDataInterfaceColorCurve
 ]
 usedDynamicInputs [&MakeVector2D, &LerpLinearColors]
 usedModules [
  &SystemState
  &self:ScratchModule_01
  &Timeline
  &self:CalculateBend
  &self:SystemSpawnScript.SystemScriptSource.SystemScriptGraph.NiagaraNodeAssignment_0.SetVariables_#1
  &self:SystemSpawnScript.SystemScriptSource.SystemScriptGraph.NiagaraNodeAssignment_1.SetVariables_#2
 ]
}
emitters {
 CenterDenseSmoke {
  emitterUpdateScript [
   &EmitterState(Life Cycle Mode=::NewEnumerator1, Loop Behavior=::NewEnumerator1, Loop Duration Mode=::NewEnumerator1, UseLoopDelay=true)
   &SpawnPerUnit(Module.Use Max Movement Threshold=true, Module.Use Movement Tolerance=false, Spawn Spacing=200, Use Spawn Probability=false) {enabled=false}
   &SpawnRate(SpawnRate=6, Use Spawn Probability=false)
   /SetVariables(Emitter.LightingCenter:Vector4f=&MakeFloat4FromVectorAndFloat("Vector (RGB)"=/Niagara/DynamicInputs/Helpers/SimulationPosition(), "Float (Alpha)"=1)) {isSetParameters=true}
  ]
  particleSpawnScript [
   &InitializeParticle(
    Color=<rgba 1 1 1 0.3>
    Lifetime=&RandomRangeFloat(Minimum=0.8, Maximum=1.3)
    Sprite Rotation Mode=::NewEnumerator2
    Sprite Size Mode=::NewEnumerator3
    Uniform Sprite Size=&Multiply_Float(
     A=&RandomRangeFloat(Minimum=220, Maximum=320)
     B=&MakeFloatFromVector(Vector=$Engine.Owner.Scale, Channel=::NewEnumerator0)
    )
    UsePositionOffset=true
    Write Material Random=true
   ) {module=InitializeParticle001}
   &ShapeLocation(
    Cylinder Height Midpoint=1
    Cylinder Radius=10
    MaskSpawnGroup=false
    Override Seed=false
    Rotation Mode=::NewEnumerator1
    Shape Primitive=::NewEnumerator1
    Surface Only=false
    UseOwnerScale=true
   )
   &AddVelocity(
    Module.Use Falloff Distance=false
    Override Seed=false
    Velocity Mode=::NewEnumerator1
    Velocity Speed=&RandomRangeFloat(Minimum=951.3323, Maximum=1305.5693)
   ) {enabled=false}
   /SetVariables(Particles.Velocity=&Multiply_Vector(A=$Particles.Velocity, B=&VectorFromFloat(Value=1))) {isSetParameters=true}
   &InheritVelocity(Inherited Velocity Speed Limit=519.3653, Speed Limit=true, UseSpeedThreshold=false)
  ]
  particleUpdateScript [
   &ParticleState()
   &CameraOffset(Camera Offset Amount=&Multiply_Float(
     A=&RandomRangeFloat(Minimum=25, Maximum=75)
     B=&FloatFromCurve(
      FloatCurve=DataInterface() {propertyValues {Curve {keys [<at 0 0.3>,<at 1 1>]}}}
      CurveIndex=/Game/Effects/Niagara/Modules/Math/FN_CameraDistance(
       0Dist:NiagaraFloat=200
       1Dist:NiagaraFloat=700
       Camera Query:NiagaraDataInterfaceCamera=DataInterface()
       Position:NiagaraPosition=$Particles.Position
      )
     )
    ))
   &ScaleSpriteSize(Scale Factor=<1.2 1>, Scale Sprite Size Mode=::NewEnumerator2) {module=ScaleSpriteSize001}
   &Drag(Use Linear Drag=true, Use Rotational Drag=false)
   &GravityForce(Gravity=<0 0 -199.0804>)
   &CurlNoiseForce(
    Bias Noise=false
    Noise Frequency=4
    Noise Strength=&Multiply_Float(A=19.596, B=$User.VelocityScale)
    Pan Noise=true
    Pan Noise Field=<0 0 2>
   )
   &PointForce(
    Force Falloff Distance=242950.5781
    Force Strength=&FloatFromCurve(
     FloatCurve=DataInterface() {propertyValues {Curve {keys [<at 0.5 0>,<at 0.7 1>]}}}
     Scale Curve=&Multiply_Float(A=0, B=$User.VelocityScale)
    )
    Module.Use Falloff=true
    Module.UseLinearForce=false
   )
   &SolveForcesAndVelocity(Clamp Velocity=false, Limit Acceleration=false)
   &ScaleSpriteSize(Scale Factor=&Vector2DFromFloat(Value=&FloatFromCurve(
      FloatCurve=DataInterface() {propertyValues {Curve {keys [{interpMode=RCIM_Cubic, leaveTangent=0.1507, tangentMode=RCTM_Break, time=0, value=0.4}, <at 1 1.2>]}}}
      Scale Curve=2
     )))
   /SetVariables(Particles.Color_ParticleEnergy:LinearColor=&ColorFromCurve(ColorCurve=DataInterface() {propertyValues {
       AlphaCurve {keys [<at 0 1 RCIM_Cubic>]}
       BlueCurve {keys [<at 0.3364 0.5>,<at 0.749 0.0667 RCIM_Cubic>]}
       GreenCurve {keys [<at 0.3364 0.0875>,<at 0.749 0 RCIM_Cubic>]}
       RedCurve {keys [<at 0.3364 0.05>,<at 0.749 0.8 RCIM_Cubic>]}
      }})) {isSetParameters=true}
   &DynamicMaterialParameters(
    Index 0 Param 1=&FloatFromCurve(FloatCurve=DataInterface() {propertyValues {Curve {keys [{tangentMode=RCTM_User, time=0, value=0}, {tangentMode=RCTM_User, time=1, value=1}]}}})
    Index 0 Param 3=&FloatFromCurve(
     FloatCurve=DataInterface() {propertyValues {Curve {keys [
         {interpMode=RCIM_Cubic, tangentMode=RCTM_Break, time=0, value=0}
         {interpMode=RCIM_Cubic, leaveTangent=-3.8707, tangentMode=RCTM_Break, time=0.2, value=1}
         {tangentMode=RCTM_User, time=1, value=0}
        ]}}}
     Scale Curve=0
    )
    Index 0 Param 4=&Multiply_Float(A=&Waveform("Waveform [1]"=::NewEnumerator12, "[1] Frequency"=13, "[1] Amplitude Min/Max"=<1 4>), B=1)
    Index 1 Param 1=&MakeFloatFromVector(Vector=&MakeVectorFromLinearColorRGB(Color=$Particles.Color_ParticleEnergy), Channel=::NewEnumerator0)
    Index 1 Param 2=&MakeFloatFromVector(Vector=&MakeVectorFromLinearColorRGB(Color=$Particles.Color_ParticleEnergy), Channel=::NewEnumerator1)
    Index 1 Param 3=&MakeFloatFromVector(Vector=&MakeVectorFromLinearColorRGB(Color=$Particles.Color_ParticleEnergy), Channel=::NewEnumerator2)
    Index 1 Param 4=&FloatFromCurve(FloatCurve=DataInterface() {propertyValues {Curve {keys [
         {interpMode=RCIM_Cubic, leaveTangent=0.1782, tangentMode=RCTM_Break, time=0, value=0.8}
         {arriveTangent=1.1875, tangentMode=RCTM_User, time=1, value=1.2}
        ]}}})
    Index 2 Param 1=100
    Index 2 Param 2=&Multiply_Float(
     A=1222.7524
     B=&FloatFromCurve(
      FloatCurve=DataInterface() {propertyValues {Curve {keys [{interpMode=RCIM_Cubic, tangentMode=RCTM_Break, time=0, value=2}, <at 0.7 1>]}}}
      CurveIndex=$Emitter.Age
     )
    )
    Index 2 Param 3=&Multiply_Float(
     A=684.907
     B=&FloatFromCurve(
      FloatCurve=DataInterface() {propertyValues {Curve {keys [{interpMode=RCIM_Cubic, tangentMode=RCTM_Break, time=0, value=2}, <at 0.7 1>]}}}
      CurveIndex=$Emitter.Age
     )
    )
    Index 2 Param 4=&FloatFromCurve(FloatCurve=DataInterface() {propertyValues {Curve {keys [
         {interpMode=RCIM_Cubic, tangentMode=RCTM_Break, time=0.5, value=0}
         {arriveTangent=2.6267, interpMode=RCIM_Cubic, tangentMode=RCTM_Break, time=1, value=0.8}
        ]}}})
    Index 3 Param 1=&FloatFromCurve(FloatCurve=DataInterface() {propertyValues {Curve {keys [<at 0 1>,<at 0.7 1>,<at 2 1>]}}}, CurveIndex=$Engine.System.Age)
    Module.Index0Param0Write=true
    Module.Index0Param1Write=true
    Module.Index0Param2Write=true
    Module.Index0Param3Write=true
    Module.Index1Param0Write=true
    Module.Index1Param1Write=true
    Module.Index1Param2Write=true
    Module.Index1Param3Write=true
    Module.Index2Param0Write=true
    Module.Index2Param1Write=true
    Module.Index2Param2Write=true
    Module.Index2Param3Write=true
    Module.Index3Param0Write=true
    Module.Index3Param1Write=true
    Module.Index3Param2Write=true
    Module.Index3Param3Write=true
    Write Parameter Index 1=true
    Write Parameter Index 2=true
    Write Parameter Index 3=true
   )
   &ScaleColor(
    Scale Alpha=&FloatFromCurve(
     FloatCurve=DataInterface() {propertyValues {Curve {keys [
         {interpMode=RCIM_Cubic, leaveTangent=1.3072, tangentMode=RCTM_Break, time=0, value=0}
         {interpMode=RCIM_Cubic, tangentMode=RCTM_Break, time=0.25, value=1}
         {arriveTangent=-4.4597, interpMode=RCIM_Cubic, leaveTangent=-4.4597, tangentMode=RCTM_User, time=1, value=0}
        ]}}}
     Scale Curve=&FN_CameraDistance(
      0Dist:NiagaraFloat=800
      1Dist:NiagaraFloat=1800
      Camera Query:NiagaraDataInterfaceCamera=DataInterface() {propertyValues {bRequireCurrentFrameData=false}}
      Position:NiagaraPosition=$Particles.Position
     )
    )
    ScaleA=true
    ScaleRGB=true
   )
  ]
  rendererClasses [&NiagaraSpriteRendererProperties]
  renderers [{rendererClass=&NiagaraSpriteRendererProperties}]
 }
 DarkEnergyWaves {
  emitterUpdateScript [
   &EmitterState(Inactive Response=::NewEnumerator1, Life Cycle Mode=::NewEnumerator1, UseLoopDelay=false)
   &SpawnBurst_Instantaneous(Spawn Count=2, Use Spawn Probability=false, UseLoopCountLimit=false)
  ]
  particleSpawnScript [
   &FN_SetSpawnGroup(Number of Spawn Groups=2)
   &InitializeParticle(
    Color=$Emitter.RandomisedColour
    Lifetime=$Emitter.CurrentLoopDuration
    Mesh Scale=&MergeHelper_VectorSwitch(0=<0.65 0.65 1.1>, 1=<0.9 0.9 1.5>, 2=?, 3=?, Index=$Particles.SpawnGroup, Num=::NewEnumerator0)
    Mesh Scale Mode=::NewEnumerator4
    Position Offset=<100 0 0>
    UsePositionOffset=true
    Write Material Random=true
   )
   /SetVariables(Particles.VisibilityTag=$Particles.SpawnGroup) {isSetParameters=true}
   &InitialMeshOrientation(Mesh Axis To Orient=<0 0 1>, Mesh Orientation Mode=::NewEnumerator2, Use Rotation Vector=false)
  ]
  particleUpdateScript [
   &ParticleState()
   &UpdateMeshOrientation(Rotation Rate=&RandomRangeFloat(Minimum=0.3, Maximum=0.75)) {enabled=false}
   &DynamicMaterialParameters(
    Index 3 Param 4=$System.CalculateBend.Angle
    Index 3 Vector 1:Vector3f=$System.CalculateBend.Axis
    Module.Index0Float0Write=false
    Module.Index0Float1Write=false
    Module.Index0Float2Write=false
    Module.Index0Float3Write=false
    Module.Index3Float3Write=true
    Module.Index3VectorWrite=true
    Write Parameter Index 3=::NewEnumerator3
   )
   &ScaleColor(
    Scale Alpha=&FN_CameraDistance(
     0Dist:NiagaraFloat=800
     1Dist:NiagaraFloat=1800
     Camera Query:NiagaraDataInterfaceCamera=DataInterface() {propertyValues {bRequireCurrentFrameData=false}}
     Position:NiagaraPosition=$Particles.Position
    )
    ScaleA=true
    ScaleRGB=true
   )
  ]
  rendererClasses [&NiagaraMeshRendererProperties]
  renderers [
   {rendererClass=&NiagaraMeshRendererProperties, rendererIndex=0}
   {rendererClass=&NiagaraMeshRendererProperties, rendererIndex=1}
  ]
 }
 DissoveParticle {
  simTarget=GPUComputeSim
  emitterUpdateScript [
   &EmitterState(Life Cycle Mode=::NewEnumerator1, Loop Behavior=::NewEnumerator1, Loop Duration Mode=::NewEnumerator1, UseLoopDelay=false)
   &SpawnRate(SpawnRate=45, Use Spawn Probability=false) {module=SpawnRate001}
   &SpawnRate(SpawnGroup=1, SpawnRate=15, Use Spawn Probability=false)
  ]
  particleSpawnScript [
   &InitializeParticle(
    Color=<rgba 0.297 0.194 0.4 1>
    Lifetime Max=3
    Lifetime Min=1.5
    Lifetime Mode=::NewEnumerator1
    Mesh Scale Mode=::NewEnumerator3
    Mesh Uniform Scale=100
    Sprite Rotation Mode=::NewEnumerator2
    Sprite Size Mode=::NewEnumerator3
    Uniform Sprite Size=&Multiply_Float(
     A=&MaskFloatBySpawnGroup(Masked Float=&RandomRangeFloat(Minimum=40, Maximum=110), Passthrough Float=185)
     B=&MakeFloatFromVector(Vector=$Engine.Owner.Scale, Channel=::NewEnumerator0)
    )
    UsePositionOffset=false
    Write Material Random=true
   )
   /SetVariables() {isSetParameters=true}
   &ShapeLocation(
    Cylinder Height=&Multiply_Float(
     A=991.4656
     B=&Power(a=&MakeFloatFromVector(Vector=$Engine.Owner.Scale, Channel=::NewEnumerator0), b=0.5)
    )
    Cylinder Height Midpoint=0.9
    Cylinder Radius=&Multiply_Float(
     A=75
     B=&Power(a=&MakeFloatFromVector(Vector=$Engine.Owner.Scale, Channel=::NewEnumerator0), b=0.5)
    )
    MaskSpawnGroup=false
    Offset=&MaskVectorBySpawnGroup(Masked Vector=<0 0 -50>, Passthrough Vector=<0 0 -50>)
    Offset Mode=::NewEnumerator0
    Override Seed=false
    Rotation Mode=::NewEnumerator1
    Shape Primitive=::NewEnumerator1
    Surface Only=false
    UseOwnerScale=true
   )
   &AddVelocity(
    Module.Use Falloff Distance=false
    Override Seed=false
    Velocity Mode=::NewEnumerator1
    Velocity Speed=&RandomRangeFloat(Minimum=598.9141, Maximum=1338.5206)
   )
   &InitialMeshOrientation(Use Rotation Vector=false) {enabled=false}
   &CameraOffset(Camera Offset Amount=-100)
  ]
  particleUpdateScript [
   &ParticleState()
   &ScaleSpriteSize(Scale Sprite Size Mode=::NewEnumerator2) {module=ScaleSpriteSize001}
   &ScaleSpriteSize(Scale Factor=&Vector2DFromFloat(Value=&FloatFromCurve(FloatCurve=DataInterface() {propertyValues {Curve {keys [{arriveTangent=-0.0952, interpMode=RCIM_Cubic, leaveTangent=-0.0952, tangentMode=RCTM_User, time=0.2, value=0.3973}]}}})))
   /Niagara/Modules/Update/SubUV/SubUVAnimation(Number Of Frames=36) {enabled=false}
   &GravityForce(Gravity=<0 0 -359.2241>)
   &WindForce(
    Wind Speed=&TransformVector(
     Vector=&RandomRangeVector(Minimum=<969.3573 0 247.7893>, Maximum=<1173.9515 1 500>)
     Source Space=::World
     Destination Space=::Simulation
    )
    Wind Speed Scale=&FloatFromCurve(
     FloatCurve=DataInterface() {propertyValues {Curve {keys [{interpMode=RCIM_Cubic, tangentMode=RCTM_Break, time=0, value=0}, {arriveTangent=2.6532, tangentMode=RCTM_User, time=1, value=1}]}}}
     CurveIndex=$Emitter.Age
    )
   ) {enabled=false}
   &CurlNoiseForce(
    Bias Noise=false
    Noise Frequency=2
    Noise Strength=&RandomRangeFloat(Minimum=1019.2568, Maximum=2091.7732)
    Pan Noise=true
    Pan Noise Field=<0 0 2>
   )
   &Drag(
    Drag=&MaskFloatBySpawnGroup(Masked Float=2, Passthrough Float=1, Spawn Group Mask=1)
    Use Linear Drag=true
    Use Rotational Drag=true
   )
   &SolveForcesAndVelocity(Clamp Velocity=false, Limit Acceleration=false)
   &ScaleColor(
    Scale Alpha=&FloatFromCurve(FloatCurve=DataInterface() {propertyValues {Curve {keys [{arriveTangent=10.7159, time=0, value=0}, {interpMode=RCIM_Cubic, tangentMode=RCTM_Break, time=0.4, value=1}]}}})
    ScaleA=true
    ScaleRGB=true
   )
   &DynamicMaterialParameters(
    Index 0 Param 1=&RandomRangeFloat(Minimum=0, Maximum=100)
    Index 0 Param 2=&FloatFromCurve(FloatCurve=DataInterface() {propertyValues {Curve {keys [<at 0 -1>,<at 1 1>]}}})
    Index 0 Param 3=&FloatFromCurve(FloatCurve=DataInterface() {propertyValues {Curve {keys [<at 0.2 1>,<at 0.4 0>]}}})
    Index 0 Param 4=1
    Module.Index0Param0Write=true
    Module.Index0Param1Write=true
    Module.Index0Param2Write=true
    Module.Index0Param3Write=true
   )
   &FN_Sprite_Attributes(
    Color=&MultiplyLinearColor(
     Linear Color=<rgba 0.3888 0.01 1 0.05>
     Scale Factor=&MakeVector4(
      X=1
      Y=1
      Z=1
      W=&FloatFromCurve(
       FloatCurve=DataInterface() {propertyValues {Curve {keys [
           {interpMode=RCIM_Cubic, tangentMode=RCTM_User, time=0, value=0}
           {interpMode=RCIM_Cubic, tangentMode=RCTM_User, time=0.1, value=1}
           {tangentMode=RCTM_User, time=0.5, value=0}
          ]}}}
       Scale Curve=0
      )
     )
    )
    Create Color Attribute=true
    Create SpriteSize Attribute=true
    SpriteSize=<400 400>
    Write to Module Namespace=true
   )
  ]
  rendererClasses [&NiagaraMeshRendererProperties, &NiagaraSpriteRendererProperties]
  renderers [
   {rendererClass=&NiagaraMeshRendererProperties, rendererIndex=0}
   {rendererClass=&NiagaraSpriteRendererProperties, rendererIndex=1}
   {rendererClass=&NiagaraSpriteRendererProperties, rendererIndex=2}
  ]
 }
 Glow1 {rendererClasses [], renderers []}
 Glow2 {rendererClasses [], renderers []}
 Glow3 {rendererClasses [], renderers []}
 Minimal {
  emitterUpdateScript [
   &EmitterState()
   &SpawnPerUnit(Module.Use Max Movement Threshold=false, Module.Use Movement Tolerance=false, Spawn Spacing=170, Use Spawn Probability=false)
   &SpawnRate(SpawnRate=500, Use Spawn Probability=false) {enabled=false}
  ]
  particleSpawnScript [
   &InitializeParticle(
    Color Mode=::NewEnumerator0
    Lifetime=&Multiply_Float(A=1, B=&MakeFloatFromVector2d(Vector 2D=$System.SpriteScale))
    Ribbon Width=&Multiply_Float(A=200, B=&MakeFloatFromVector2d(Vector 2D=$System.SpriteScale))
    Ribbon Width Mode=::NewEnumerator1
    UsePositionOffset=false
    Write Material Random=true
   )
   &AddVelocity(Velocity=<-1000 0 0>) {enabled=false}
  ]
  particleUpdateScript [
   &ParticleState()
   &SolveForcesAndVelocity(Clamp Velocity=false, Limit Acceleration=false)
   &ScaleRibbonWidth(Ribbon Width Scale=&FloatFromCurve(FloatCurve=DataInterface() {propertyValues {Curve {
        defaultValue=1
        keys [
         {arriveTangent=2.662, interpMode=RCIM_Cubic, leaveTangent=2.662, tangentMode=RCTM_User, time=0, value=0.7}
         {arriveTangent=0.0606, interpMode=RCIM_Cubic, leaveTangent=0.0606, tangentMode=RCTM_User, time=0.2, value=0.9847}
        ]
       }}}))
   &ScaleColor(Scale Alpha=$System.Timeline.Time, ScaleA=true, ScaleRGB=true)
   &ScaleColor(
    Scale Alpha=&FN_CameraDistance(
     0Dist:NiagaraFloat=99999
     1Dist:NiagaraFloat=999999
     Camera Query:NiagaraDataInterfaceCamera=DataInterface() {propertyValues {bRequireCurrentFrameData=false}}
     Position:NiagaraPosition=$Particles.Position
    )
    ScaleA=true
    ScaleRGB=true
   ) {enabled=false, module=ScaleColor001}
   &DynamicMaterialParameters(
    Index 0 Param 1=&FN_CameraDistance(
     0Dist:NiagaraFloat=800
     1Dist:NiagaraFloat=1800
     Camera Query:NiagaraDataInterfaceCamera=DataInterface() {propertyValues {bRequireCurrentFrameData=false}}
     Position:NiagaraPosition=$Particles.Position
    )
    Module.Index0Float0Write=true
    Module.Index0Float1Write=true
    Module.Index0Float2Write=true
    Module.Index0Float3Write=true
   )
  ]
  rendererClasses [&NiagaraRibbonRendererProperties]
  renderers [{rendererClass=&NiagaraRibbonRendererProperties}]
 }
 Minimal001 {
  emitterUpdateScript [
   &EmitterState()
   &SpawnRate(SpawnRate=500, Use Spawn Probability=false) {enabled=false}
   &SpawnPerUnit(Module.Use Max Movement Threshold=false, Module.Use Movement Tolerance=false, Spawn Spacing=170, Use Spawn Probability=false)
  ]
  particleSpawnScript [
   &InitializeParticle(
    Color=<rgba 0.05 0.025 0.05 1>
    Lifetime=&Multiply_Float(A=0.7, B=&MakeFloatFromVector2d(Vector 2D=$System.SpriteScale))
    Ribbon Width=&Multiply_Float(A=315, B=&MakeFloatFromVector2d(Vector 2D=$System.SpriteScale))
    Ribbon Width Mode=::NewEnumerator1
    UsePositionOffset=false
    Write Material Random=true
   )
   &AddVelocity(Velocity=<-1000 0 0>) {enabled=false}
  ]
  particleUpdateScript [
   &ParticleState()
   &SolveForcesAndVelocity(Clamp Velocity=false, Limit Acceleration=false)
   &ScaleRibbonWidth(Ribbon Width Scale=&FloatFromCurve(FloatCurve=DataInterface() {propertyValues {Curve {keys [
         {interpMode=RCIM_Cubic, leaveTangent=0.6153, tangentMode=RCTM_Break, time=0, value=1.2}
         {arriveTangent=0.1614, tangentMode=RCTM_User, time=1, value=1.5}
        ]}}}))
   &ScaleColor(
    Scale Alpha=&FloatFromCurve(
     FloatCurve=DataInterface() {propertyValues {Curve {keys [
         {interpMode=RCIM_Cubic, leaveTangent=-6.5884, tangentMode=RCTM_Break, time=0, value=2}
         {arriveTangent=0.0905, interpMode=RCIM_Cubic, leaveTangent=0.0905, tangentMode=RCTM_User, time=0.4, value=1}
         <at 1 1>
        ]}}}
     Scale Curve=&Multiply_Float(
      A=&Subtract_Float(A=1, B=$System.Timeline.Time)
      B=&FN_CameraDistance(
       0Dist:NiagaraFloat=800
       1Dist:NiagaraFloat=1800
       Camera Query:NiagaraDataInterfaceCamera=DataInterface() {propertyValues {bRequireCurrentFrameData=false}}
       Position:NiagaraPosition=$Particles.Position
      )
     )
    )
    ScaleA=true
    ScaleRGB=true
   )
   &DynamicMaterialParameters(
    Index 0 Param 1=-0.7
    Index 0 Param 3=1
    Index 0 Param 4=1
    Module.Index0Float0Write=true
    Module.Index0Float1Write=true
    Module.Index0Float2Write=true
    Module.Index0Float3Write=true
   )
   /SetVariables(
    Particles.Color2=&MultiplyLinearColor(
     Linear Color=<rgba 0.125 0 0.15 20>
     Scale Factor=&MakeVector4FromVectorAndFloat(
      "Vector (RGB)"=<1 1 1>
      "Float (Alpha)"=&Multiply_Float(
       A=&Subtract_Float(A=1, B=$System.Timeline.Time)
       B=&FN_CameraDistance(
        0Dist:NiagaraFloat=800
        1Dist:NiagaraFloat=1800
        Camera Query:NiagaraDataInterfaceCamera=DataInterface() {propertyValues {bRequireCurrentFrameData=false}}
        Position:NiagaraPosition=$Particles.Position
       )
      )
     )
    )
    Particles.RibbonWidth2:NiagaraFloat=&Multiply_Float(
     A=&Multiply_Float(A=223, B=&MakeFloatFromVector2d(Vector 2D=$System.SpriteScale))
     B=&FloatFromCurve(FloatCurve=DataInterface() {propertyValues {Curve {keys [<at 0 1.3>,<at 1 1>]}}})
    )
   ) {isSetParameters=true}
   /SetVariables(
    Particles.Color3:LinearColor=&MultiplyLinearColor(
     Linear Color=<rgba 0.125 0 0.15 1>
     Scale Factor=&MakeVector4FromVectorAndFloat(
      "Vector (RGB)"=<1 1 1>
      "Float (Alpha)"=&Multiply_Float(
       A=&Subtract_Float(A=1, B=$System.Timeline.Time)
       B=&FN_CameraDistance(
        0Dist:NiagaraFloat=800
        1Dist:NiagaraFloat=1800
        Camera Query:NiagaraDataInterfaceCamera=DataInterface() {propertyValues {bRequireCurrentFrameData=false}}
        Position:NiagaraPosition=$Particles.Position
       )
      )
     )
    )
    Particles.RibbonWidth3:NiagaraFloat=&Multiply_Float(
     A=&Multiply_Float(A=300, B=&MakeFloatFromVector2d(Vector 2D=$System.SpriteScale))
     B=&FloatFromCurve(FloatCurve=DataInterface() {propertyValues {Curve {keys [{interpMode=RCIM_Cubic, leaveTangent=1.8403, tangentMode=RCTM_Break, time=0, value=0.3}, <at 0.2968 1.1>, <at 0.8379 0.7>]}}})
    )
   ) {isSetParameters=true}
  ]
  rendererClasses [&NiagaraRibbonRendererProperties]
  renderers [
   {rendererClass=&NiagaraRibbonRendererProperties}
   {rendererClass=&NiagaraRibbonRendererProperties, rendererIndex=1}
   {rendererClass=&NiagaraRibbonRendererProperties, rendererIndex=2}
  ]
 }
 Minimal002 {
  bEnabled=false
  emitterUpdateScript [
   &EmitterState()
   &SpawnRate(SpawnRate=500, Use Spawn Probability=false) {enabled=false}
   &SpawnPerUnit(Module.Use Max Movement Threshold=false, Module.Use Movement Tolerance=false, Spawn Spacing=140, Use Spawn Probability=false)
  ]
  particleSpawnScript [
   &InitializeParticle(
    Color=<rgba 0.125 0 0.15 20>
    Lifetime=&Multiply_Float(A=0.7, B=&MakeFloatFromVector2d(Vector 2D=$System.SpriteScale))
    Ribbon Width=&Multiply_Float(A=223, B=&MakeFloatFromVector2d(Vector 2D=$System.SpriteScale))
    Ribbon Width Mode=::NewEnumerator1
    UsePositionOffset=false
    Write Material Random=true
   )
   &AddVelocity(Velocity=<-1000 0 0>) {enabled=false}
  ]
  particleUpdateScript [
   &ParticleState()
   &SolveForcesAndVelocity(Clamp Velocity=false, Limit Acceleration=false)
   &ScaleRibbonWidth(Ribbon Width Scale=&FloatFromCurve(FloatCurve=DataInterface() {propertyValues {Curve {keys [<at 0 1.3>,<at 1 1>]}}}))
   &DynamicMaterialParameters(Module.Index0Float0Write=true, Module.Index0Float1Write=true, Module.Index0Float2Write=true, Module.Index0Float3Write=true)
   &ScaleColor(Scale Alpha=&Subtract_Float(A=1, B=$System.Timeline.Time), ScaleA=true, ScaleRGB=true)
  ]
  rendererClasses [&NiagaraRibbonRendererProperties]
  renderers [{rendererClass=&NiagaraRibbonRendererProperties}]
 }
 Minimal003 {
  bEnabled=false
  emitterUpdateScript [
   &EmitterState()
   &SpawnRate(SpawnRate=500, Use Spawn Probability=false) {enabled=false}
   &SpawnPerUnit(Module.Use Max Movement Threshold=false, Module.Use Movement Tolerance=false, Spawn Spacing=140, Use Spawn Probability=false)
  ]
  particleSpawnScript [
   &InitializeParticle(
    Color=<rgba 0.125 0 0.15 1>
    Lifetime=&Multiply_Float(A=0.55, B=&MakeFloatFromVector2d(Vector 2D=$System.SpriteScale))
    Position Offset=<100 0 0>
    Ribbon Width=&Multiply_Float(A=300, B=&MakeFloatFromVector2d(Vector 2D=$System.SpriteScale))
    Ribbon Width Mode=::NewEnumerator1
    UsePositionOffset=false
    Write Material Random=true
   )
   &AddVelocity(Velocity=<-1000 0 0>) {enabled=false}
  ]
  particleUpdateScript [
   &ParticleState()
   &SolveForcesAndVelocity(Clamp Velocity=false, Limit Acceleration=false)
   &ScaleRibbonWidth(Ribbon Width Scale=&FloatFromCurve(FloatCurve=DataInterface() {propertyValues {Curve {keys [{interpMode=RCIM_Cubic, leaveTangent=1.8403, tangentMode=RCTM_Break, time=0, value=0.3}, <at 0.3633 1.1>, <at 1 0.7>]}}}))
   &DynamicMaterialParameters(Module.Index0Float0Write=true, Module.Index0Float1Write=true, Module.Index0Float2Write=true, Module.Index0Float3Write=true)
   &ScaleColor(Scale Alpha=&Subtract_Float(A=1, B=$System.Timeline.Time), ScaleA=true, ScaleRGB=true)
  ]
  rendererClasses [&NiagaraRibbonRendererProperties]
  renderers [{rendererClass=&NiagaraRibbonRendererProperties}]
 }
 Pendulum {
  bEnabled=false
  emitterUpdateScript [
   &EmitterState(
    Inactive Response=::NewEnumerator1
    Life Cycle Mode=::NewEnumerator1
    Loop Behavior=::NewEnumerator1
    Loop Duration Mode=::NewEnumerator1
    UseLoopDelay=false
   )
   &SpawnBurst_Instantaneous(Use Spawn Probability=false, UseLoopCountLimit=false)
  ]
  particleSpawnScript [&InitializeParticle(Color Mode=::NewEnumerator0, Sprite Size Mode=::NewEnumerator3, Uniform Sprite Size=500, UsePositionOffset=false, Write Material Random=true)]
  particleUpdateScript [
   &ParticleState(Kill Particles When Lifetime Has Elapsed=false, Let Infinitely Lived Particles Die When Emitter Deactivates=true)
   &PendulumSetup(
    Module.Enable Spring Driver=true
    Pendulum Length=200
    Pendulum Pivot Offset Coordinate Space:ENiagaraCoordinateSpace=::Local
    Pendulum Rest Axis:Vector3f=&TransformVector(Vector=<-1 0 0>, Destination Space=::World)
    "Spring Driven Constraint (Tightness)"=1
   ) {enabled=false}
   &SolveForcesAndVelocity(Clamp Velocity=false, Limit Acceleration=false)
   &PendulumConstraint() {enabled=false}
   &CalculateAccurateVelocity() {enabled=false}
   &InterpolateOverTime(
    Initial Position Value:NiagaraPosition=$Engine.Emitter.SimulationPosition
    Rate Of Change=2
    Target Position Value:NiagaraPosition=$Engine.Emitter.SimulationPosition
    Variable Type=::NewEnumerator6
   )
   /SetVariables(Particles.Position=$Particles.InterpolateOverTime.InterpolateOverTime_Position.Moving Average) {isSetParameters=true}
  ]
  rendererClasses [&NiagaraSpriteRendererProperties]
  renderers [{rendererClass=&NiagaraSpriteRendererProperties}]
 }
}
summary {
 systemName=NS_Projectile_EventQueen
 emitters [
  {emitterName=DarkEnergyWaves, rendererClasses [&NiagaraMeshRendererProperties]}
  {emitterName=Glow2, rendererClasses []}
  {emitterName=Glow3, rendererClasses []}
  {emitterName=Glow1, rendererClasses []}
  {emitterName=Minimal, rendererClasses [&NiagaraRibbonRendererProperties]}
  {emitterName=Minimal001, rendererClasses [&NiagaraRibbonRendererProperties]}
  {bEnabled=false, emitterName=Minimal002, rendererClasses [&NiagaraRibbonRendererProperties]}
  {bEnabled=false, emitterName=Minimal003, rendererClasses [&NiagaraRibbonRendererProperties]}
  {
   emitterName=DissoveParticle
   simTarget=GPUComputeSim
   rendererClasses [&NiagaraMeshRendererProperties, &NiagaraSpriteRendererProperties]
  }
  {emitterName=CenterDenseSmoke, rendererClasses [&NiagaraSpriteRendererProperties]}
  {bEnabled=false, emitterName=Pendulum, rendererClasses [&NiagaraSpriteRendererProperties]}
 ]
 userVariables [
  User.bIsFrontEnd {type=NiagaraBool}
  User.bIsFrontEndPreview {type=NiagaraBool}
  User.ElementSizeScale {type=NiagaraFloat, defaultValue=1, description=""}
  User.SmokeColorMul {type=LinearColor, description="", defaultValue {struct=LinearColor, value=<rgba 1 1 1 1>}}
  User.SmokeSizeScale {type=NiagaraFloat, defaultValue=1, description=""}
  User.VelocityScale {type=NiagaraFloat, defaultValue=1, description=""}
 ]
}
