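# Niagara effect for the event variant of the TidalCrane swarm boss, named as a projectile muzzle-ready cue: a magenta-violet glow.
# Its single ParticleDecal emitter burst-spawns into three sprite renderers and a light renderer, pushed toward the camera and offset along Z.
# Curves, random ranges and waveforms animate size, color and alpha, also writing separate custom decal and flare attributes for color and size alongside the light attributes.
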NiagaraSystem /TidalCrane_Swarm_Boss/Assets/FX/EventVariant/Projectile/NS_TidalCrane_Event_Projectile_MuzzelReady
dependencies {
 usedDataInterfaces [&NiagaraDataInterfaceCurve]
 usedDynamicInputs [&MakeVector4_FromLinearColor]
 usedModules [&SystemState]
}
emitters {ParticleDecal {
  emitterUpdateScript [
   &EmitterState()
   &SpawnBurst_Instantaneous(Use Spawn Probability=false, UseLoopCountLimit=false)
  ]
  particleSpawnScript [
   &InitializeParticle(
    Color=<rgba 1.8931 0.0195 2 1>
    Sprite Size Mode=::NewEnumerator3
    Uniform Sprite Size=2000
    UsePositionOffset=false
    Write Material Random=true
   )
   &DynamicMaterialParameters(
    Index 0 Param 1=1
    Index 0 Param 2=1
    Module.Index0Float0Write=true
    Module.Index0Float1Write=true
    Module.Index0Float2Write=true
    Module.Index0Float3Write=true
   )
   /SetVariables(Particles.EnergyDP:Vector4f=<2 1 8 1>) {isSetParameters=true}
   &CameraOffset(Camera Offset Amount=1500)
  ]
  particleUpdateScript [
   &ParticleState()
   &ScaleSpriteSize(
    Scale Sprite Size Mode=::NewEnumerator0
    Uniform Scale Factor=&Multiply_Float(
     A=&RandomRangeFloat(Minimum=0.7, Maximum=1, Evaluation Type=::NewEnumerator1)
     B=&Lerp_Float(
      A=0
      B=&Multiply_Float(
       A=&Lerp_Float(
        A=&RandomRangeFloat(Minimum=0.5, Maximum=1.5, Evaluation Type=::NewEnumerator1)
        B=1
        Alpha=&FloatFromCurve(FloatCurve=DataInterface() {propertyValues {Curve {
            defaultValue=0
            keys [
             {arriveTangent=-7.8658, interpMode=RCIM_Cubic, leaveTangent=-7.8658, leaveTangentWeight=0.5627, tangentMode=RCTM_User, time=0.6, value=1}
             {arriveTangentWeight=0.0392, interpMode=RCIM_Cubic, tangentMode=RCTM_User, time=1, value=0}
            ]
           }}})
       )
       B=2
      )
      Alpha=&FloatFromCurve(FloatCurve=DataInterface() {propertyValues {Curve {
          defaultValue=0
          keys [{arriveTangent=14.3269, interpMode=RCIM_Cubic, leaveTangent=14.3269, tangentMode=RCTM_User, time=0, value=0}, <at 0.1175 0.8606 RCIM_Cubic>]
         }}})
     )
    )
   ) {module=ScaleSpriteSize002}
   &ScaleColor(
    Scale Alpha=&FloatFromCurve(
     FloatCurve=DataInterface() {propertyValues {Curve {
        defaultValue=1
        keys [
         {interpMode=RCIM_Cubic, tangentMode=RCTM_Break, time=0, value=0}
         {arriveTangent=-0.0405, interpMode=RCIM_Cubic, leaveTangent=-0.0405, tangentMode=RCTM_Break, time=0.4577, value=0.9869}
         {arriveTangent=-4.2895, interpMode=RCIM_Cubic, leaveTangent=-4.2895, tangentMode=RCTM_User, time=1, value=0}
        ]
       }}}
     Scale Curve=&Waveform("Waveform [1]"=::NewEnumerator0, "[1] Phase Scale"=2, "[1] Frequency"=1, "[1] Amplitude Min/Max"=<0 1>)
    )
    Scale RGB=&VectorFromFloat(Value=&FloatFromCurve(
      FloatCurve=DataInterface() {propertyValues {Curve {
         defaultValue=1
         keys [
          {interpMode=RCIM_Cubic, tangentMode=RCTM_Break, time=0, value=0}
          {arriveTangent=-0.0405, interpMode=RCIM_Cubic, leaveTangent=-0.0405, tangentMode=RCTM_Break, time=0.7095, value=1}
         ]
        }}}
      Scale Curve=0.3
     ))
    ScaleA=true
    ScaleRGB=true
   )
   /SetVariables(
    Particles.ParticleDecalColor:LinearColor=&Multiply_LinearColorByFloat(Linear Color=$Particles.Color, Float=5)
    Particles.ParticleDecalSize:Vector2f=&Multiply_Vector2D_ByFloat(Float=0.5, Vector2D=$Particles.SpriteSize)
   ) {isSetParameters=true}
   /SetVariables(
    Particles.FlareColor=&Multiply_LinearColorByFloat(
     Linear Color=<rgba 1.095 0.1611 2 1>
     Float=&Multiply_Float(
      A=&Lerp_Float(
       A=0
       B=1
       Alpha=&FloatFromCurve(
        FloatCurve=DataInterface() {propertyValues {Curve {
           defaultValue=0
           keys [
            {interpMode=RCIM_Cubic, tangentMode=RCTM_Break, time=0, value=0}
            {arriveTangent=-0.0405, interpMode=RCIM_Cubic, leaveTangent=-0.0405, tangentMode=RCTM_User, time=0.7009, value=1}
            {arriveTangent=-8.0089, interpMode=RCIM_Cubic, leaveTangent=-8.0089, tangentMode=RCTM_User, time=1, value=0}
           ]
          }}}
        Scale Curve=0.7
       )
      )
      B=&Waveform("Waveform [1]"=::NewEnumerator0, "[1] Phase Scale"=2, "[1] Frequency"=1, "[1] Amplitude Min/Max"=<0 1>)
     )
    )
    Particles.FlareSize:Vector2f=&Multiply_Vector2D_ByFloat(
     Float=&Multiply_Float(
      A=&RandomRangeFloat(Minimum=0.7, Maximum=1, Evaluation Type=::NewEnumerator1)
      B=&Lerp_Float(
       A=0
       B=&Multiply_Float(
        A=&Lerp_Float(
         A=&RandomRangeFloat(Minimum=0.5, Maximum=1, Evaluation Type=::NewEnumerator1)
         B=1
         Alpha=&FloatFromCurve(FloatCurve=DataInterface() {propertyValues {Curve {
             defaultValue=0
             keys [
              {arriveTangent=-7.8658, interpMode=RCIM_Cubic, leaveTangent=-7.8658, leaveTangentWeight=0.5627, tangentMode=RCTM_User, time=0.6, value=1}
              {arriveTangentWeight=0.0392, interpMode=RCIM_Cubic, tangentMode=RCTM_User, time=1, value=0}
             ]
            }}})
        )
        B=1
       )
       Alpha=&FloatFromCurve(
        FloatCurve=DataInterface() {propertyValues {Curve {
           defaultValue=0
           keys [{arriveTangent=14.3269, interpMode=RCIM_Cubic, leaveTangent=14.3269, tangentMode=RCTM_User, time=0, value=0}, <at 0.1175 0.8606 RCIM_Cubic>]
          }}}
        Scale Curve=3
       )
      )
     )
     Vector2D=<2000 1000>
    )
   ) {isSetParameters=true}
   &Light_Attributes(
    Color=1
    Color Value To Scale:LinearColor=$Particles.Color
    Light Position Offset:Vector3f=<0 0 250>
    Light Radius=3000
    Position=1
    Position Offset Coordinate Space:ENiagaraCoordinateSpace=::Local
    Radius=1
    Scale RGB:Vector3f=&VectorFromFloat(Value=50)
    ScaleA=true
    ScaleRGB=true
    UsePositionOffset=true
   )
   &OffsetPosition(Position Offset=&MakeVector(
     X=0
     Y=0
     Z=&Lerp_Float(
      A=173.2242
      B=458.7671
      Alpha=&FloatFromCurve(FloatCurve=DataInterface() {propertyValues {Curve {keys [<at 0 0>,<at 1 1>]}}})
     )
    ))
  ]
  rendererClasses [&NiagaraSpriteRendererProperties, &NiagaraLightRendererProperties]
  renderers [
   {rendererClass=&NiagaraSpriteRendererProperties}
   {rendererClass=&NiagaraSpriteRendererProperties, rendererIndex=1}
   {rendererClass=&NiagaraLightRendererProperties, rendererIndex=2}
   {rendererClass=&NiagaraSpriteRendererProperties, rendererIndex=3}
  ]
 }}
summary {
 systemName=NS_TidalCrane_Event_Projectile_MuzzelReady
 emitters [{emitterName=ParticleDecal, rendererClasses [&NiagaraSpriteRendererProperties, &NiagaraLightRendererProperties]}]
 userVariables []
}
